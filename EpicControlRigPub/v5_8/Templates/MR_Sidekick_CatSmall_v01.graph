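# Modular Control Rig template for a small cat Sidekick character, holding its quadruped skeleton, curves and module connectors.
# Connectors define modules for root, body, spine, neck, face, ears, tail, four legs, three emote props and IK-bone pinning.
# The hierarchy includes detailed face bones, IK foot bones and curves for facial poses, pupils, scaling and morphs.
# The model graph contains only an empty Begin Execution event, with no hand-wired logic.
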ControlRigBlueprint /EpicControlRigPub/v5_8/Templates/MR_Sidekick_CatSmall_v01
graphs [
 RigVMFunctionLibrary {links [], nodes []}
 RigVMModel {nodes [RigUnit_BeginExecution = &RigUnit_BeginExecution()]}
]
hierarchy [
 BothLegIK {type=Curve, loc=<0 0 0>, rot=<0 0 0 1>}
 L_eye_look_up_pose {type=Curve, loc=<0 0 0>, rot=<0 0 0 1>}
 L_wide_pose {type=Curve, loc=<0 0 0>, rot=<0 0 0 1>}
 R_lip_corner_narrow_pose {type=Curve, loc=<0 0 0>, rot=<0 0 0 1>}
 R_upper_lip_lower_pose {type=Curve, loc=<0 0 0>, rot=<0 0 0 1>}
 L_lip_corner_narrow_pose {type=Curve, loc=<0 0 0>, rot=<0 0 0 1>}
 L_upper_lip_lower_pose {type=Curve, loc=<0 0 0>, rot=<0 0 0 1>}
 R_lower_lip_up_pose {type=Curve, loc=<0 0 0>, rot=<0 0 0 1>}
 R_brow_squeeze_in_pose {type=Curve, loc=<0 0 0>, rot=<0 0 0 1>}
 L_lip_corner_wide_pose {type=Curve, loc=<0 0 0>, rot=<0 0 0 1>}
 L_upperLidDown_pose {type=Curve, loc=<0 0 0>, rot=<0 0 0 1>}
 R_lower_lip_down_pose {type=Curve, loc=<0 0 0>, rot=<0 0 0 1>}
 R_upper_lip_raiser_pose {type=Curve, loc=<0 0 0>, rot=<0 0 0 1>}
 L_lower_lip_down_pose {type=Curve, loc=<0 0 0>, rot=<0 0 0 1>}
 L_upper_lip_raiser_pose {type=Curve, loc=<0 0 0>, rot=<0 0 0 1>}
 L_brow_squeeze_in_pose {type=Curve, loc=<0 0 0>, rot=<0 0 0 1>}
 L_lower_lip_up_pose {type=Curve, loc=<0 0 0>, rot=<0 0 0 1>}
 R_upperLidDown_pose {type=Curve, loc=<0 0 0>, rot=<0 0 0 1>}
 R_wide_pose {type=Curve, loc=<0 0 0>, rot=<0 0 0 1>}
 R_blink_pose {type=Curve, loc=<0 0 0>, rot=<0 0 0 1>}
 L_eye_look_left_pose {type=Curve, loc=<0 0 0>, rot=<0 0 0 1>}
 L_eye_look_down_pose {type=Curve, loc=<0 0 0>, rot=<0 0 0 1>}
 L_blink_pose {type=Curve, loc=<0 0 0>, rot=<0 0 0 1>}
 Reaction_Happy {type=Curve, loc=<0 0 0>, rot=<0 0 0 1>}
 L_eye_look_right_pose {type=Curve, loc=<0 0 0>, rot=<0 0 0 1>}
 R_lip_corner_wide_pose {type=Curve, loc=<0 0 0>, rot=<0 0 0 1>}
 L_smile_pose {type=Curve, loc=<0 0 0>, rot=<0 0 0 1>}
 R_smile_pose {type=Curve, loc=<0 0 0>, rot=<0 0 0 1>}
 L_frown_pose {type=Curve, loc=<0 0 0>, rot=<0 0 0 1>}
 R_frown_pose {type=Curve, loc=<0 0 0>, rot=<0 0 0 1>}
 C_nose_wrinkler_pose {type=Curve, loc=<0 0 0>, rot=<0 0 0 1>}
 L_cheek_up_pose {type=Curve, loc=<0 0 0>, rot=<0 0 0 1>}
 R_cheek_up_pose {type=Curve, loc=<0 0 0>, rot=<0 0 0 1>}
 jaw_open_pose {type=Curve, loc=<0 0 0>, rot=<0 0 0 1>}
 jaw_left_pose {type=Curve, loc=<0 0 0>, rot=<0 0 0 1>}
 jaw_right_pose {type=Curve, loc=<0 0 0>, rot=<0 0 0 1>}
 jaw_up_pose {type=Curve, loc=<0 0 0>, rot=<0 0 0 1>}
 jaw_forward_pose {type=Curve, loc=<0 0 0>, rot=<0 0 0 1>}
 jaw_back_pose {type=Curve, loc=<0 0 0>, rot=<0 0 0 1>}
 eye_convergence_pose {type=Curve, loc=<0 0 0>, rot=<0 0 0 1>}
 tongue_forward_pose {type=Curve, loc=<0 0 0>, rot=<0 0 0 1>}
 tongue_back_pose {type=Curve, loc=<0 0 0>, rot=<0 0 0 1>}
 tongue_up_pose {type=Curve, loc=<0 0 0>, rot=<0 0 0 1>}
 tongue_down_pose {type=Curve, loc=<0 0 0>, rot=<0 0 0 1>}
 L_brow_In_Up_pose {type=Curve, loc=<0 0 0>, rot=<0 0 0 1>}
 L_brow_Ot_Up_pose {type=Curve, loc=<0 0 0>, rot=<0 0 0 1>}
 L_brow_In_Dn_pose {type=Curve, loc=<0 0 0>, rot=<0 0 0 1>}
 L_brow_Ot_Dn_pose {type=Curve, loc=<0 0 0>, rot=<0 0 0 1>}
 R_brow_In_Up_pose {type=Curve, loc=<0 0 0>, rot=<0 0 0 1>}
 R_brow_Ot_Up_pose {type=Curve, loc=<0 0 0>, rot=<0 0 0 1>}
 R_brow_In_Dn_pose {type=Curve, loc=<0 0 0>, rot=<0 0 0 1>}
 R_brow_Ot_Dn_pose {type=Curve, loc=<0 0 0>, rot=<0 0 0 1>}
 upper_lip_up_pose {type=Curve, loc=<0 0 0>, rot=<0 0 0 1>}
 upper_lip_dn_pose {type=Curve, loc=<0 0 0>, rot=<0 0 0 1>}
 lower_lip_up_pose {type=Curve, loc=<0 0 0>, rot=<0 0 0 1>}
 lower_lip_dn_pose {type=Curve, loc=<0 0 0>, rot=<0 0 0 1>}
 L_smile_open_correction_pose {type=Curve, loc=<0 0 0>, rot=<0 0 0 1>}
 R_smile_open_correction_pose {type=Curve, loc=<0 0 0>, rot=<0 0 0 1>}
 tongue_left_pose {type=Curve, loc=<0 0 0>, rot=<0 0 0 1>}
 tongue_right_pose {type=Curve, loc=<0 0 0>, rot=<0 0 0 1>}
 reverse_jaw_pose {type=Curve, loc=<0 0 0>, rot=<0 0 0 1>}
 base_pose {type=Curve, loc=<0 0 0>, rot=<0 0 0 1>}
 R_eye_look_left_pose {type=Curve, loc=<0 0 0>, rot=<0 0 0 1>}
 R_eye_look_right_pose {type=Curve, loc=<0 0 0>, rot=<0 0 0 1>}
 R_eye_look_up_pose {type=Curve, loc=<0 0 0>, rot=<0 0 0 1>}
 R_eye_look_down_pose {type=Curve, loc=<0 0 0>, rot=<0 0 0 1>}
 L_lower_lid_up_pose {type=Curve, loc=<0 0 0>, rot=<0 0 0 1>}
 R_lower_lid_up_pose {type=Curve, loc=<0 0 0>, rot=<0 0 0 1>}
 tongue_curl_tip_up_pose {type=Curve, loc=<0 0 0>, rot=<0 0 0 1>}
 tongue_curl_tip_down_pose {type=Curve, loc=<0 0 0>, rot=<0 0 0 1>}
 tongue_curl_mid_up_pose {type=Curve, loc=<0 0 0>, rot=<0 0 0 1>}
 tongue_curl_mid_down_pose {type=Curve, loc=<0 0 0>, rot=<0 0 0 1>}
 L_tongue_twist_pose {type=Curve, loc=<0 0 0>, rot=<0 0 0 1>}
 R_tongue_twist_pose {type=Curve, loc=<0 0 0>, rot=<0 0 0 1>}
 AttachEmote1InPelvisSpace {type=Curve, loc=<0 0 0>, rot=<0 0 0 1>}
 AttachEmote2InPelvisSpace {type=Curve, loc=<0 0 0>, rot=<0 0 0 1>}
 AttachEmote3InPelvisSpace {type=Curve, loc=<0 0 0>, rot=<0 0 0 1>}
 AnimateAcc {type=Curve, loc=<0 0 0>, rot=<0 0 0 1>}
 pupil_shape {type=Curve, loc=<0 0 0>, rot=<0 0 0 1>}
 pupil_size {type=Curve, loc=<0 0 0>, rot=<0 0 0 1>}
 bypass_poseAsset {type=Curve, loc=<0 0 0>, rot=<0 0 0 1>}
 Body_Scale {type=Curve, loc=<0 0 0>, rot=<0 0 0 1>}
 BodyMorph_LOD0 {type=Curve, loc=<0 0 0>, rot=<0 0 0 1>}
 HoodieMorph_LOD0 {type=Curve, loc=<0 0 0>, rot=<0 0 0 1>}
 BodyScale {type=Curve, loc=<0 0 0>, rot=<0 0 0 1>}
 root {loc=<0 0 0>, rot=<0 0 0 1>}
 pelvis {parent=root, loc=<0 -14.664 26.7066>, rot=<0.5008 0.4992 -0.5008 -0.4992>}
 spine_01 {parent=pelvis, loc=<0 0 0>, rot=<0 0 -0.0124 -0.9999>}
 spine_02 {parent=spine_01, loc=<14.2907 -0.3558 0>, rot=<0 0 0.0159 -0.9999>}
 spine_03 {parent=spine_02, loc=<13.1068 0 0>, rot=<0 0 0.0041 -1>}
 neck_01 {parent=spine_03, loc=<2.5947 -1.5005 0>, rot=<0 0 0.4032 -0.9151>}
 neck_02 {parent=neck_01, loc=<5.711 0 0>, rot=<0 0 0.0462 -0.9989>}
 head {parent=neck_02, loc=<6.5839 0 0>, rot=<-0.2129 -0.6743 0.2129 -0.6743>}
 face_attach {parent=head, loc=<0 -0.8148 -0.4208>, rot=<0 0 0 1>}
 jaw {parent=face_attach, loc=<0 1.0725 1.4273>, rot=<0 0.7071 0 -0.7071>}
 lip_lower_03_l {parent=jaw, loc=<-4.0483 8.5878 -3.1509>, rot=<-0.0721 -0.6559 -0.2295 -0.7155>}
 lip_lower_02_l {parent=jaw, loc=<-4.3078 9.7759 -2.1948>, rot=<-0.2406 -0.6832 -0.3562 -0.5904>}
 lip_lower_01_l {parent=jaw, loc=<-4.4416 10.7595 -1.1753>, rot=<-0.3576 -0.704 -0.3942 -0.4702>}
 lip_lower_mid {parent=jaw, loc=<-4.2223 11.0052 0>, rot=<0.3881 0.5911 0.5911 0.3881>}
 lip_lower_01_r {parent=jaw, loc=<-4.4416 10.7595 1.1753>, rot=<0.704 -0.3576 -0.4702 0.3942>}
 lip_lower_02_r {parent=jaw, loc=<-4.3078 9.776 2.1948>, rot=<0.6832 -0.2406 -0.5904 0.3562>}
 lip_lower_03_r {parent=jaw, loc=<-4.0483 8.5878 3.1509>, rot=<-0.6559 0.0721 0.7155 -0.2295>}
 teeth_lower {parent=jaw, loc=<-4.1786 7.958 0>, rot=<0 -0.7071 0 -0.7071>}
 tongue_01 {parent=jaw, loc=<-1.2582 3.3853 0>, rot=<0.6017 -0.3714 0.3714 -0.6017>}
 tongue_02 {parent=tongue_01, loc=<-2.4201 0 0>, rot=<-0.0005 -0.062 0.0001 -0.9981>}
 tongue_03 {parent=tongue_02, loc=<-2.0927 0 0>, rot=<0.0005 -0.0094 -0.0001 -1>}
 lip_corner_l {parent=face_attach, loc=<4.2193 8.3992 -1.8726>, rot=<0.0176 -0.0132 -0.1977 -0.98>}
 cheek_bt_l {parent=face_attach, loc=<6.5575 5.747 -1.6239>, rot=<0.0809 -0.1726 -0.1547 -0.9694>}
 cheek_bt_r {parent=face_attach, loc=<-6.5575 5.7469 -1.6239>, rot=<0.9694 -0.1547 0.1726 0.0809>}
 reverse_jaw {parent=face_attach, loc=<0 0.0433 0.7402>, rot=<0 0.7071 0 -0.7071>}
 lip_upper_mid {parent=reverse_jaw, loc=<-1.6808 13.1404 0>, rot=<-0.4557 -0.5407 -0.5407 -0.4557>}
 lip_upper_01_l {parent=reverse_jaw, loc=<-2.0571 12.7201 -1.4511>, rot=<-0.4899 -0.6512 -0.3566 -0.457>}
 lip_upper_02_l {parent=reverse_jaw, loc=<-2.6632 11.7206 -2.8105>, rot=<-0.2406 -0.6832 -0.3562 -0.5904>}
 lip_upper_03_l {parent=reverse_jaw, loc=<-2.22 10.3784 -3.582>, rot=<-0.0167 -0.7319 -0.2603 -0.6295>}
 cheek_tp_in_l {parent=reverse_jaw, loc=<1.1042 11.3628 -2.3015>, rot=<-0.4855 -0.6882 -0.4984 -0.2056>}
 cheek_tp_ot_l {parent=reverse_jaw, loc=<1.3732 7.6666 -7.0875>, rot=<-0.1961 -0.6424 -0.2197 -0.7076>}
 cheek_tp_in_r {parent=reverse_jaw, loc=<1.1042 11.3628 2.3015>, rot=<0.6882 -0.4855 -0.2056 0.4984>}
 cheek_tp_ot_r {parent=reverse_jaw, loc=<1.3732 7.6666 7.0875>, rot=<-0.6424 0.1961 0.7076 -0.2197>}
 brow_ot_r {parent=reverse_jaw, loc=<5.3307 8.8034 6.2208>, rot=<-0.5915 -0.3152 0.633 0.3875>}
 brow_md_r {parent=reverse_jaw, loc=<6.1238 9.743 4.46>, rot=<0.7483 0.0798 -0.6522 -0.0916>}
 brow_ct {parent=reverse_jaw, loc=<5.7814 10.2512 0>, rot=<0 -0.7071 0 -0.7071>}
 brow_md_l {parent=reverse_jaw, loc=<6.1237 9.7429 -4.46>, rot=<0.0798 -0.7483 0.0916 -0.6522>}
 brow_in_l {parent=reverse_jaw, loc=<5.9564 9.9456 -2.1248>, rot=<0.1061 -0.6991 0.1061 -0.6991>}
 brow_ot_l {parent=reverse_jaw, loc=<5.3307 8.8034 -6.2208>, rot=<0.3152 -0.5915 0.3875 -0.633>}
 brow_in_r {parent=reverse_jaw, loc=<5.9564 9.9456 2.1248>, rot=<-0.6991 -0.1061 0.6991 0.1061>}
 eye_r {parent=reverse_jaw, loc=<3.7889 7.8018 3.6467>, rot=<-0.7012 -0.0583 0.708 0.0598>}
 lid_bt_r {parent=reverse_jaw, loc=<3.7883 7.778 3.6467>, rot=<-0.6966 -0.0705 0.7124 0.0476>}
 lid_tp_r {parent=reverse_jaw, loc=<3.7889 7.8018 3.6467>, rot=<-0.6966 -0.0705 0.7124 0.0476>}
 nose_mid {parent=reverse_jaw, loc=<0.4554 13.7896 0>, rot=<0.4878 0.5119 0.5119 0.4878>}
 nostril_l {parent=nose_mid, loc=<-0.0539 -0.8279 -0.434>, rot=<0 0 0 -1>}
 nostril_r {parent=nose_mid, loc=<-0.0539 0.8279 -0.4339>, rot=<0 -1 0 0>}
 lid_tp_l {parent=reverse_jaw, loc=<3.7889 7.8019 -3.6467>, rot=<0.0705 -0.6966 0.0476 -0.7124>}
 lid_bt_l {parent=reverse_jaw, loc=<3.7883 7.7781 -3.6467>, rot=<0.0705 -0.6966 0.0476 -0.7124>}
 eye_l {parent=reverse_jaw, loc=<3.7889 7.8019 -3.6467>, rot=<0.0583 -0.7012 0.0598 -0.708>}
 lip_upper_01_r {parent=reverse_jaw, loc=<-2.0571 12.7202 1.4512>, rot=<0.6512 -0.4899 -0.457 0.3566>}
 lip_upper_02_r {parent=reverse_jaw, loc=<-2.6633 11.7206 2.8105>, rot=<0.6832 -0.2406 -0.5904 0.3562>}
 lip_upper_03_r {parent=reverse_jaw, loc=<-2.22 10.3784 3.582>, rot=<0.7319 -0.0167 -0.6295 0.2603>}
 teeth_upper {parent=reverse_jaw, loc=<-0.8072 9.0408 0>, rot=<0 -0.7071 0 -0.7071>}
 ear_01_l {parent=reverse_jaw, loc=<1.5571 0.0575 -0.5535>, rot=<-0.0464 -0.3542 0.0211 -0.9338>}
 ear_02_l {parent=ear_01_l, loc=<7.9141 0 0>, rot=<0.119 0.0062 -0.1268 -0.9847>}
 ear_03_l {parent=ear_02_l, loc=<2.3632 0 0>, rot=<0.0137 0.0739 -0.0375 -0.9965>}
 ear_04_l {parent=ear_03_l, loc=<2.608 0 0>, rot=<-0.0159 -0.0128 -0.0425 -0.9989>}
 ear_01_r {parent=reverse_jaw, loc=<1.5571 0.0575 0.5535>, rot=<-0.3542 0.0464 0.9338 0.0211>}
 ear_02_r {parent=ear_01_r, loc=<-7.914 0 0>, rot=<0.119 0.0062 -0.1268 -0.9847>}
 ear_03_r {parent=ear_02_r, loc=<-2.3632 -0.0001 0>, rot=<0.0137 0.0739 -0.0375 -0.9965>}
 ear_04_r {parent=ear_03_r, loc=<-2.6079 0.0001 -0.0001>, rot=<-0.0159 -0.0128 -0.0425 -0.9989>}
 lip_corner_r {parent=face_attach, loc=<-4.2193 8.3993 -1.8726>, rot=<0.98 -0.1977 0.0132 0.0176>}
 shoulder_fr_l {parent=spine_03, loc=<1.9267 -0.6483 -5.7705>, rot=<0.0135 -0.0015 -0.7949 -0.6066>}
 elbow_fr_l {parent=shoulder_fr_l, loc=<10.8758 0 0>, rot=<-0.0045 0.004 0.1185 -0.9929>}
 wrist_fr_l {parent=elbow_fr_l, loc=<11.4815 0 0>, rot=<0.9838 -0.1787 0.0143 -0.005>}
 ball_fr_l {parent=wrist_fr_l, loc=<3.7813 0 0>, rot=<-0.0436 0.0565 -0.4345 -0.8978>}
 toe_fr_l {parent=ball_fr_l, loc=<1.4234 0 0>, rot=<-0.0136 -0.0343 0.0437 -0.9984>}
 toe_tip_fr_l {parent=toe_fr_l, loc=<3.1113 0 0>, rot=<0.0031 0.006 -0.1897 -0.9818>}
 scapula_fr_l {parent=shoulder_fr_l, loc=<-5.0362 3.7723 -0.1058>, rot=<0.9487 -0.3158 0.0121 -0.0097>}
 shoulder_fr_r {parent=spine_03, loc=<1.9267 -0.6483 5.7706>, rot=<-0.0015 -0.0135 0.6066 -0.7949>}
 elbow_fr_r {parent=shoulder_fr_r, loc=<-10.8758 0 -0.0001>, rot=<-0.0045 0.004 0.1185 -0.9929>}
 wrist_fr_r {parent=elbow_fr_r, loc=<-11.4814 0 0>, rot=<0.9838 -0.1787 0.0143 -0.005>}
 ball_fr_r {parent=wrist_fr_r, loc=<-3.7813 0 0>, rot=<-0.0436 0.0565 -0.4345 -0.8978>}
 toe_fr_r {parent=ball_fr_r, loc=<-1.4235 0 0>, rot=<-0.0136 -0.0343 0.0437 -0.9984>}
 toe_tip_fr_r {parent=toe_fr_r, loc=<-3.1112 0 0.0001>, rot=<0.0031 0.006 -0.1897 -0.9818>}
 scapula_fr_r {parent=shoulder_fr_r, loc=<5.0362 -3.7723 0.1057>, rot=<0.9487 -0.3158 0.0121 -0.0097>}
 dyn_R_clothingSleeve_upper {parent=spine_03, loc=<1.0313 4.6907 6.3502>, rot=<-0.497 -0.503 0.497 -0.503>}
 dyn_L_clothingSleeve_upper {parent=spine_03, loc=<1.0313 4.6907 -6.3502>, rot=<0.503 -0.497 -0.503 -0.497>}
 L_buckle {parent=spine_02, loc=<5.6998 6.7783 -8.1402>, rot=<-0.4991 -0.5009 0.4991 -0.5009>}
 R_buckle {parent=spine_02, loc=<5.6998 6.7783 8.1402>, rot=<0.5009 -0.4991 -0.5009 -0.4991>}
 hips {parent=pelvis, loc=<0 0 0>, rot=<0 0 0 1>}
 thigh_bk_l {parent=hips, loc=<-1.7864 -5.2948 -5.6956>, rot=<0.7238 0.6899 -0.0073 0.0076>}
 knee_bk_l {parent=thigh_bk_l, loc=<12.276 0 0>, rot=<0.0064 -0.005 0.4145 -0.91>}
 ankle_bk_l {parent=knee_bk_l, loc=<11.9381 0 0>, rot=<-0.0046 0.0019 -0.5264 -0.8502>}
 ball_bk_l {parent=ankle_bk_l, loc=<10.3665 0 0>, rot=<-0.0049 -0.0045 -0.4831 -0.8756>}
 toe_bk_l {parent=ball_bk_l, loc=<1.8315 0 0>, rot=<0 0 0.0737 -0.9973>}
 toe_tip_bk_l {parent=toe_bk_l, loc=<2.4192 0 0>, rot=<0 0 -0.2485 -0.9686>}
 thigh_bk_r {parent=hips, loc=<-1.7864 -5.2948 5.6956>, rot=<0.6899 -0.7238 -0.0076 -0.0073>}
 knee_bk_r {parent=thigh_bk_r, loc=<-12.276 0 0>, rot=<0.0064 -0.005 0.4145 -0.91>}
 ankle_bk_r {parent=knee_bk_r, loc=<-11.9381 0 0>, rot=<-0.0046 0.0019 -0.5264 -0.8502>}
 ball_bk_r {parent=ankle_bk_r, loc=<-10.3664 0 0>, rot=<-0.0049 -0.0044 -0.4831 -0.8756>}
 toe_bk_r {parent=ball_bk_r, loc=<-1.8315 0 0>, rot=<0 0 0.0737 -0.9973>}
 toe_tip_bk_r {parent=toe_bk_r, loc=<-2.4192 0 0>, rot=<0 0 -0.2485 -0.9686>}
 tail_01 {parent=hips, loc=<-8.3781 -5.5168 0>, rot=<0.015 -0.7069 0.7069 -0.015>}
 tail_02 {parent=tail_01, loc=<6.6572 0 0>, rot=<0 -0.0169 0 -0.9999>}
 tail_03 {parent=tail_02, loc=<6.6562 0 0>, rot=<0 -0.009 0 -1>}
 tail_04 {parent=tail_03, loc=<6.6565 0 0>, rot=<0 0.0272 0 -0.9996>}
 tail_05 {parent=tail_04, loc=<6.6561 0 0>, rot=<0 -0.0025 0 -1>}
 IK_Foot_Rear_Root {parent=root, loc=<0 -19.464 0>, rot=<0 0 0 1>}
 IK_Foot_Rear_L {parent=IK_Foot_Rear_Root, loc=<6.2133 -2.0165 1.6094>, rot=<-0.4299 -0.5614 -0.4299 -0.5614>}
 IK_Foot_Rear_R {parent=IK_Foot_Rear_Root, loc=<-6.2133 -2.0165 1.6094>, rot=<0.5614 -0.4299 0.5614 -0.4299>}
 IK_Foot_Front_Root {parent=root, loc=<0 1.5578 0>, rot=<0 0 0 1>}
 IK_Foot_Front_L {parent=IK_Foot_Front_Root, loc=<6.1013 11.4199 1.8214>, rot=<-0.45 -0.5454 -0.3709 -0.602>}
 IK_Foot_Front_R {parent=IK_Foot_Front_Root, loc=<-6.1013 11.4199 1.8214>, rot=<0.602 -0.3709 0.5454 -0.45>}
 attach_emote_01 {parent=root, loc=<0 0 0>, rot=<0 0 0 1>}
 attach_emote_02 {parent=root, loc=<0 0 0>, rot=<0 0 0 1>}
 attach_emote_03 {parent=root, loc=<0 0 0>, rot=<0 0 0 1>}
 AccessoryScale {type=Curve, loc=<0 0 0>, rot=<0 0 0 1>}
 root/RootJoint {type=Connector, loc=<0 0 0>, rot=<0 0 0 1>}
 root/Body {type=Connector, loc=<0 0 0>, rot=<0 0 0 1>}
 Capsule/Root {type=Connector, loc=<0 0 0>, rot=<0 0 0 1>}
 Emote Prop 1/Start {type=Connector, loc=<0 0 0>, rot=<0 0 0 1>}
 Emote Prop 1/Parent {type=Connector, loc=<0 0 0>, rot=<0 0 0 1>}
 Emote Prop 1/End {type=Connector, loc=<0 0 0>, rot=<0 0 0 1>}
 Emote Prop 1/Orient Spaces {type=Connector, loc=<0 0 0>, rot=<0 0 0 1>}
 Emote Prop 2/Start {type=Connector, loc=<0 0 0>, rot=<0 0 0 1>}
 Emote Prop 2/Parent {type=Connector, loc=<0 0 0>, rot=<0 0 0 1>}
 Emote Prop 2/End {type=Connector, loc=<0 0 0>, rot=<0 0 0 1>}
 Emote Prop 2/Orient Spaces {type=Connector, loc=<0 0 0>, rot=<0 0 0 1>}
 Emote Prop 3/Start {type=Connector, loc=<0 0 0>, rot=<0 0 0 1>}
 Emote Prop 3/Parent {type=Connector, loc=<0 0 0>, rot=<0 0 0 1>}
 Emote Prop 3/End {type=Connector, loc=<0 0 0>, rot=<0 0 0 1>}
 Emote Prop 3/Orient Spaces {type=Connector, loc=<0 0 0>, rot=<0 0 0 1>}
 Body/Parent {type=Connector, loc=<0 0 0>, rot=<0 0 0 1>}
 Body/Body {type=Connector, loc=<0 0 0>, rot=<0 0 0 1>}
 Body/Right Hip {type=Connector, loc=<0 0 0>, rot=<0 0 0 1>}
 Body/Left Hip {type=Connector, loc=<0 0 0>, rot=<0 0 0 1>}
 Spine/SpineRoot {type=Connector, loc=<0 0 0>, rot=<0 0 0 1>}
 Spine/Spine 3 {type=Connector, loc=<0 0 0>, rot=<0 0 0 1>}
 Spine/Spine 1 {type=Connector, loc=<0 0 0>, rot=<0 0 0 1>}
 Spine/Parent {type=Connector, loc=<0 0 0>, rot=<0 0 0 1>}
 Spine/Spine 2 {type=Connector, loc=<0 0 0>, rot=<0 0 0 1>}
 Neck/Root {type=Connector, loc=<0 0 0>, rot=<0 0 0 1>}
 Neck/Head {type=Connector, loc=<0 0 0>, rot=<0 0 0 1>}
 Neck/NeckTip {type=Connector, loc=<0 0 0>, rot=<0 0 0 1>}
 Neck/Parent {type=Connector, loc=<0 0 0>, rot=<0 0 0 1>}
 Neck/NeckSpaces {type=Connector, loc=<0 0 0>, rot=<0 0 0 1>}
 Neck/HeadSpaces {type=Connector, loc=<0 0 0>, rot=<0 0 0 1>}
 Face/Head {type=Connector, loc=<0 0 0>, rot=<0 0 0 1>}
 Ear Root L/Start {type=Connector, loc=<0 0 0>, rot=<0 0 0 1>}
 Ear Root L/Parent {type=Connector, loc=<0 0 0>, rot=<0 0 0 1>}
 Ear Root L/End {type=Connector, loc=<0 0 0>, rot=<0 0 0 1>}
 Ear Root L/Orient Spaces {type=Connector, loc=<0 0 0>, rot=<0 0 0 1>}
 Ear L/Start {type=Connector, loc=<0 0 0>, rot=<0 0 0 1>}
 Ear L/Parent {type=Connector, loc=<0 0 0>, rot=<0 0 0 1>}
 Ear L/End {type=Connector, loc=<0 0 0>, rot=<0 0 0 1>}
 Ear L/Orient Spaces {type=Connector, loc=<0 0 0>, rot=<0 0 0 1>}
 Ear Root R/Start {type=Connector, loc=<0 0 0>, rot=<0 0 0 1>}
 Ear Root R/Parent {type=Connector, loc=<0 0 0>, rot=<0 0 0 1>}
 Ear Root R/End {type=Connector, loc=<0 0 0>, rot=<0 0 0 1>}
 Ear Root R/Orient Spaces {type=Connector, loc=<0 0 0>, rot=<0 0 0 1>}
 Ear R/Start {type=Connector, loc=<0 0 0>, rot=<0 0 0 1>}
 Ear R/Parent {type=Connector, loc=<0 0 0>, rot=<0 0 0 1>}
 Ear R/End {type=Connector, loc=<0 0 0>, rot=<0 0 0 1>}
 Ear R/Orient Spaces {type=Connector, loc=<0 0 0>, rot=<0 0 0 1>}
 Tail/HipRoot {type=Connector, loc=<0 0 0>, rot=<0 0 0 1>}
 Tail/Tail {type=Connector, loc=<0 0 0>, rot=<0 0 0 1>}
 Tail/Parent {type=Connector, loc=<0 0 0>, rot=<0 0 0 1>}
 Tail/Orient Spaces {type=Connector, loc=<0 0 0>, rot=<0 0 0 1>}
 Back Leg L/Start {type=Connector, loc=<0 0 0>, rot=<0 0 0 1>}
 Back Leg L/Mid {type=Connector, loc=<0 0 0>, rot=<0 0 0 1>}
 Back Leg L/End {type=Connector, loc=<0 0 0>, rot=<0 0 0 1>}
 Back Leg L/Ball {type=Connector, loc=<0 0 0>, rot=<0 0 0 1>}
 Back Leg L/Toe {type=Connector, loc=<0 0 0>, rot=<0 0 0 1>}
 Back Leg L/Toe Tip {type=Connector, loc=<0 0 0>, rot=<0 0 0 1>}
 Back Leg L/Parent {type=Connector, loc=<0 0 0>, rot=<0 0 0 1>}
 Back Leg L/IK Parent {type=Connector, loc=<0 0 0>, rot=<0 0 0 1>}
 Back Leg L/IK Spaces {type=Connector, loc=<0 0 0>, rot=<0 0 0 1>}
 Back Leg L/Scapula {type=Connector, loc=<0 0 0>, rot=<0 0 0 1>}
 Front Leg L/Start {type=Connector, loc=<0 0 0>, rot=<0 0 0 1>}
 Front Leg L/Mid {type=Connector, loc=<0 0 0>, rot=<0 0 0 1>}
 Front Leg L/End {type=Connector, loc=<0 0 0>, rot=<0 0 0 1>}
 Front Leg L/Ball {type=Connector, loc=<0 0 0>, rot=<0 0 0 1>}
 Front Leg L/Toe {type=Connector, loc=<0 0 0>, rot=<0 0 0 1>}
 Front Leg L/Toe Tip {type=Connector, loc=<0 0 0>, rot=<0 0 0 1>}
 Front Leg L/Parent {type=Connector, loc=<0 0 0>, rot=<0 0 0 1>}
 Front Leg L/IK Parent {type=Connector, loc=<0 0 0>, rot=<0 0 0 1>}
 Front Leg L/IK Spaces {type=Connector, loc=<0 0 0>, rot=<0 0 0 1>}
 Front Leg L/Scapula {type=Connector, loc=<0 0 0>, rot=<0 0 0 1>}
 Back Leg R/Start {type=Connector, loc=<0 0 0>, rot=<0 0 0 1>}
 Back Leg R/Mid {type=Connector, loc=<0 0 0>, rot=<0 0 0 1>}
 Back Leg R/End {type=Connector, loc=<0 0 0>, rot=<0 0 0 1>}
 Back Leg R/Ball {type=Connector, loc=<0 0 0>, rot=<0 0 0 1>}
 Back Leg R/Toe {type=Connector, loc=<0 0 0>, rot=<0 0 0 1>}
 Back Leg R/Toe Tip {type=Connector, loc=<0 0 0>, rot=<0 0 0 1>}
 Back Leg R/Parent {type=Connector, loc=<0 0 0>, rot=<0 0 0 1>}
 Back Leg R/IK Parent {type=Connector, loc=<0 0 0>, rot=<0 0 0 1>}
 Back Leg R/IK Spaces {type=Connector, loc=<0 0 0>, rot=<0 0 0 1>}
 Back Leg R/Scapula {type=Connector, loc=<0 0 0>, rot=<0 0 0 1>}
 Front Leg R/Start {type=Connector, loc=<0 0 0>, rot=<0 0 0 1>}
 Front Leg R/Mid {type=Connector, loc=<0 0 0>, rot=<0 0 0 1>}
 Front Leg R/End {type=Connector, loc=<0 0 0>, rot=<0 0 0 1>}
 Front Leg R/Ball {type=Connector, loc=<0 0 0>, rot=<0 0 0 1>}
 Front Leg R/Toe {type=Connector, loc=<0 0 0>, rot=<0 0 0 1>}
 Front Leg R/Toe Tip {type=Connector, loc=<0 0 0>, rot=<0 0 0 1>}
 Front Leg R/Parent {type=Connector, loc=<0 0 0>, rot=<0 0 0 1>}
 Front Leg R/IK Parent {type=Connector, loc=<0 0 0>, rot=<0 0 0 1>}
 Front Leg R/IK Spaces {type=Connector, loc=<0 0 0>, rot=<0 0 0 1>}
 Front Leg R/Scapula {type=Connector, loc=<0 0 0>, rot=<0 0 0 1>}
 Pin IK Bones/Root {type=Connector, loc=<0 0 0>, rot=<0 0 0 1>}
 Pin IK Bones/Drivers {type=Connector, loc=<0 0 0>, rot=<0 0 0 1>}
 Pin IK Bones/Driven {type=Connector, loc=<0 0 0>, rot=<0 0 0 1>}
]
