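# Modular-style Control Rig template for a large dog sidekick, holding the quadruped skeleton with face, ear, tail, IK foot and attach bones.
# Its graph has only an empty Begin Execution node, so the file shows no solve logic of its own.
# Namespaced connectors define module slots for body, spine, neck, face, ears, tail, four legs, emote props and IK-bone pinning.
# Curves carry facial poses, pupil shaping and attachment-space switches.
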ControlRigBlueprint /EpicControlRigPub/v5_8/Templates/MR_Sidekick_DogLarge_v01
graphs [
 RigVMFunctionLibrary {links [], nodes []}
 RigVMModel {nodes [RigUnit_BeginExecution = &RigUnit_BeginExecution()]}
]
hierarchy [
 root {loc=<0 0 0>, rot=<0 0 0 1>}
 pelvis {parent=root, loc=<0 -14.6653 38.9513>, rot=<0.5006 0.4994 -0.5006 -0.4994>}
 spine_01 {parent=pelvis, loc=<0 0 0>, rot=<0 0 -0.0124 -0.9999>}
 spine_02 {parent=spine_01, loc=<15.529 -0.3867 0>, rot=<0 0 0.0156 -0.9999>}
 spine_03 {parent=spine_02, loc=<14.3456 0 0>, rot=<0 0 0.004 -1>}
 neck_01 {parent=spine_03, loc=<0.7833 0.3499 0>, rot=<0 0 0.5128 -0.8585>}
 neck_02 {parent=neck_01, loc=<12.1968 0 0>, rot=<0 0 0.0313 -0.9995>}
 head {parent=neck_02, loc=<9.6064 0 0>, rot=<-0.1386 -0.6934 0.1386 -0.6934>}
 face_attach {parent=head, loc=<0 0 0>, rot=<0 0 0 1>}
 jaw {parent=face_attach, loc=<0 1.257 2.4215>, rot=<0 0.7071 0 -0.7071>}
 lip_lower_03_l {parent=jaw, loc=<-7.4496 7.0207 -4.261>, rot=<-0.0671 -0.7112 -0.0827 -0.6948>}
 lip_lower_02_l {parent=jaw, loc=<-7.2595 9.7274 -3.5139>, rot=<-0.1806 -0.7258 -0.1279 -0.6514>}
 lip_lower_01_l {parent=jaw, loc=<-6.9353 12.7416 -2.2731>, rot=<-0.1667 -0.7326 -0.1369 -0.6456>}
 lip_lower_01_r {parent=jaw, loc=<-6.9353 12.7417 2.2731>, rot=<0.7326 -0.1667 -0.6456 0.1369>}
 lip_lower_02_r {parent=jaw, loc=<-7.2595 9.7274 3.5139>, rot=<0.7258 -0.1806 -0.6514 0.1279>}
 lip_lower_03_r {parent=jaw, loc=<-7.4496 7.0207 4.261>, rot=<0.7112 -0.0671 -0.6948 0.0827>}
 lip_lower_mid {parent=jaw, loc=<-7.06 13.8318 0>, rot=<0.4939 0.506 0.506 0.4939>}
 teeth_lower {parent=jaw, loc=<-6.7086 6.2372 0>, rot=<0 -0.7071 0 -0.7071>}
 tongue_01 {parent=jaw, loc=<-4.6681 0.7455 0>, rot=<0.5253 -0.4734 0.4734 -0.5253>}
 tongue_02 {parent=tongue_01, loc=<-5.0324 -0.0138 0.1501>, rot=<-0.0051 0.0564 -0.0009 -0.9984>}
 tongue_03 {parent=tongue_02, loc=<-3.7171 0.0226 0.4386>, rot=<0.0051 -0.0111 0.0006 -0.9999>}
 lip_corner_l {parent=face_attach, loc=<4.6401 7.4463 -4.1217>, rot=<0.0187 -0.0115 -0.106 -0.9941>}
 lip_corner_r {parent=face_attach, loc=<-4.6401 7.4463 -4.1217>, rot=<0.9941 -0.106 0.0115 0.0187>}
 cheek_bt_l {parent=face_attach, loc=<7.7862 2.8351 -3.0651>, rot=<0.0975 -0.0174 -0.1751 -0.9796>}
 cheek_bt_r {parent=face_attach, loc=<-7.7862 2.8351 -3.0651>, rot=<0.9796 -0.1751 0.0174 0.0975>}
 reverse_jaw {parent=face_attach, loc=<0 2.7286 -3.1927>, rot=<0 0.7071 0 -0.7071>}
 lip_upper_01_l {parent=reverse_jaw, loc=<1.3684 12.3381 -3.1339>, rot=<-0.2403 -0.6929 -0.2602 -0.628>}
 lip_upper_02_l {parent=reverse_jaw, loc=<1.2751 8.5527 -4.2666>, rot=<-0.1806 -0.7258 -0.1279 -0.6514>}
 lip_upper_03_l {parent=reverse_jaw, loc=<1.0041 5.9897 -5.1335>, rot=<-0.0671 -0.7112 -0.0827 -0.6948>}
 brow_ot_r {parent=reverse_jaw, loc=<10.2921 2.3975 6.1241>, rot=<-0.6792 -0.1966 0.6792 0.1966>}
 brow_md_r {parent=reverse_jaw, loc=<11.6637 3.9036 4.1437>, rot=<-0.6803 -0.1929 0.6803 0.1929>}
 brow_ct {parent=reverse_jaw, loc=<11.7248 4.3074 0>, rot=<0.1061 -0.6991 0.1061 -0.6991>}
 brow_md_l {parent=reverse_jaw, loc=<11.6637 3.9036 -4.1437>, rot=<0.1929 -0.6803 0.1929 -0.6803>}
 brow_in_l {parent=reverse_jaw, loc=<12.0101 4.7652 -2.0511>, rot=<0.1061 -0.6991 0.1061 -0.6991>}
 brow_ot_l {parent=reverse_jaw, loc=<10.2921 2.3976 -6.1241>, rot=<0.1966 -0.6792 0.1966 -0.6792>}
 brow_in_r {parent=reverse_jaw, loc=<12.0101 4.7653 2.0511>, rot=<-0.6991 -0.1061 0.6991 0.1061>}
 eye_r {parent=reverse_jaw, loc=<8.4394 2.7735 3.4163>, rot=<0.693 0.1403 -0.6874 -0.1657>}
 eye_rSocket {parent=eye_r, type=Null, loc=<0 -2.1411 0>, rot=<0 0 0 1>}
 nose_mid {parent=reverse_jaw, loc=<4.0171 14.2245 0>, rot=<-0.546 -0.4493 -0.4493 -0.546>}
 nostril_l {parent=nose_mid, loc=<0.6716 -1.4444 -0.0035>, rot=<0 0 0 -1>}
 nostril_r {parent=nose_mid, loc=<0.6716 1.4444 -0.0035>, rot=<0 -1 0 0>}
 nose_midSocket {parent=nose_mid, type=Null, loc=<1.3482 0 0.7575>, rot=<0 0 0 1>}
 eye_l {parent=reverse_jaw, loc=<8.4394 2.7736 -3.4163>, rot=<0.1403 -0.693 0.1657 -0.6874>}
 eye_lSocket {parent=eye_l, type=Null, loc=<0 2.0925 0>, rot=<0 0 0 1>}
 lip_upper_01_r {parent=reverse_jaw, loc=<1.3683 12.3382 3.134>, rot=<0.6929 -0.2403 -0.628 0.2602>}
 lip_upper_02_r {parent=reverse_jaw, loc=<1.275 8.5527 4.2666>, rot=<0.7258 -0.1806 -0.6514 0.1279>}
 lip_upper_03_r {parent=reverse_jaw, loc=<1.0041 5.9897 5.1335>, rot=<0.7112 -0.0671 -0.6948 0.0827>}
 ear_01_l {parent=reverse_jaw, loc=<6.4278 -2.8708 -1.4184>, rot=<0.0182 -0.1188 0.0334 -0.9922>}
 ear_02_l {parent=ear_01_l, loc=<6.4266 -1.2899 -2.1608>, rot=<0 0 0 1>}
 ear_03_l {parent=ear_02_l, loc=<3.908 0.0691 -0.731>, rot=<0 0 0 1>}
 ear_03_lSocket {parent=ear_03_l, type=Null, loc=<5.5072 0 0>, rot=<0 0 0 1>}
 ear_in_l {parent=ear_02_l, loc=<4.2352 2.0162 1.8546>, rot=<0 0 0 1>}
 ear_ot_l {parent=ear_02_l, loc=<3.4353 -0.115 -2.9256>, rot=<0 0 0 1>}
 ear_01_r {parent=reverse_jaw, loc=<6.4278 -2.8708 1.4184>, rot=<-0.1188 -0.0182 0.9922 0.0334>}
 ear_02_r {parent=ear_01_r, loc=<-6.4266 1.2899 2.1608>, rot=<0 0 0 -1>}
 ear_03_r {parent=ear_02_r, loc=<-3.908 -0.0692 0.731>, rot=<0 0 0 -1>}
 ear_03_rSocket {parent=ear_03_r, type=Null, loc=<-4.968 0 -0.9058>, rot=<0 0 0 1>}
 ear_ot_r {parent=ear_02_r, loc=<-3.4353 0.115 2.9256>, rot=<0 0 0 1>}
 ear_in_r {parent=ear_02_r, loc=<-4.2352 -2.0162 -1.8546>, rot=<0 0 0 1>}
 lip_upper_mid {parent=reverse_jaw, loc=<1.8989 14.219 0>, rot=<-0.4824 -0.517 -0.517 -0.4824>}
 cheek_tp_in_l {parent=reverse_jaw, loc=<5.7551 6.6606 -2.4082>, rot=<-0.2594 -0.6712 -0.2639 -0.6423>}
 cheek_tp_ot_l {parent=reverse_jaw, loc=<4.9733 2.7698 -7.0218>, rot=<-0.1358 -0.6932 -0.1692 -0.6874>}
 cheek_tp_in_r {parent=reverse_jaw, loc=<5.7551 6.6606 2.4082>, rot=<0.6712 -0.2594 -0.6423 0.2639>}
 cheek_tp_ot_r {parent=reverse_jaw, loc=<4.9733 2.7698 7.0218>, rot=<0.6932 -0.1358 -0.6874 0.1692>}
 lid_bt_r {parent=reverse_jaw, loc=<8.4394 2.7735 3.4163>, rot=<-0.6903 -0.153 0.6903 0.153>}
 lid_tp_r {parent=reverse_jaw, loc=<8.4394 2.7735 3.4163>, rot=<-0.6903 -0.153 0.6903 0.153>}
 lid_tp_l {parent=reverse_jaw, loc=<8.4394 2.7736 -3.4163>, rot=<0.153 -0.6903 0.153 -0.6903>}
 lid_bt_l {parent=reverse_jaw, loc=<8.4394 2.7736 -3.4163>, rot=<0.153 -0.6903 0.153 -0.6903>}
 teeth_upper {parent=reverse_jaw, loc=<2.6971 5.3069 0>, rot=<0 -0.7071 0 -0.7071>}
 headSocket {parent=head, type=Null, loc=<0 4.1427 10.8575>, rot=<0 0 0 1>}
 shoulder_fr_l {parent=spine_03, loc=<3.3143 1.7539 -7.578>, rot=<0.0042 0.003 -0.8375 -0.5464>}
 elbow_fr_l {parent=shoulder_fr_l, loc=<16.0381 0 0>, rot=<0.0002 0.0027 0.2036 -0.9791>}
 wrist_fr_l {parent=elbow_fr_l, loc=<15.7305 0 0>, rot=<0.9928 -0.1202 0 0>}
 ball_fr_l {parent=wrist_fr_l, loc=<4.0643 0 0>, rot=<0.0014 0.0007 -0.4114 -0.9115>}
 toe_fr_l {parent=ball_fr_l, loc=<2.6599 0 0>, rot=<-0.0027 -0.0108 -0.1114 -0.9937>}
 toe_tip_fr_l {parent=toe_fr_l, loc=<4.2353 0 0>, rot=<0.0035 0.0046 -0.1104 -0.9939>}
 ball_fr_lSocket {parent=ball_fr_l, type=Null, loc=<1.6306 0 0>, rot=<0 0 0 1>}
 scapula_fr_l {parent=shoulder_fr_l, loc=<-5.958 6.6503 -0.4062>, rot=<-0.0056 0.0274 0.4074 -0.9128>}
 shoulder_fr_r {parent=spine_03, loc=<3.3143 1.7539 7.578>, rot=<0.003 -0.0042 0.5464 -0.8375>}
 elbow_fr_r {parent=shoulder_fr_r, loc=<-16.0381 0 0>, rot=<0.0002 0.0027 0.2036 -0.9791>}
 wrist_fr_r {parent=elbow_fr_r, loc=<-15.7306 0.0001 0>, rot=<0.9928 -0.1202 0 0>}
 ball_fr_r {parent=wrist_fr_r, loc=<-4.0643 0.0001 0>, rot=<0.0014 0.0007 -0.4114 -0.9115>}
 toe_fr_r {parent=ball_fr_r, loc=<-2.6599 0 0>, rot=<-0.0027 -0.0108 -0.1114 -0.9937>}
 toe_tip_fr_r {parent=toe_fr_r, loc=<-4.2353 0 0>, rot=<0.0035 0.0046 -0.1104 -0.9939>}
 ball_fr_rSocket {parent=ball_fr_r, type=Null, loc=<-2.0884 0 0>, rot=<0 0 0 1>}
 scapula_fr_r {parent=shoulder_fr_r, loc=<5.9579 -6.6503 0.4062>, rot=<-0.0056 0.0274 0.4074 -0.9128>}
 attach_accessory {parent=spine_03, loc=<12.8014 -0.3993 0>, rot=<-0.497 -0.503 0.497 -0.503>}
 attach_backpack {parent=spine_02, loc=<14.3456 0 0>, rot=<0 0 0.004 -1>}
 attach_cape {parent=spine_02, loc=<14.3456 0 0>, rot=<0 0 0.004 -1>}
 hips {parent=pelvis, loc=<0 0 0>, rot=<0 0 0 1>}
 thigh_bk_l {parent=hips, loc=<-3.6025 0.5607 -6.2814>, rot=<-0.6817 -0.7316 0 0>}
 knee_bk_l {parent=thigh_bk_l, loc=<16.6492 0 0>, rot=<0 0 0.4212 -0.907>}
 ankle_bk_l {parent=knee_bk_l, loc=<12.9953 0 0>, rot=<0 0 -0.4271 -0.9042>}
 ball_bk_l {parent=ankle_bk_l, loc=<11.8273 0 0>, rot=<0 0 -0.5311 -0.8473>}
 toe_bk_l {parent=ball_bk_l, loc=<2.1609 0 0>, rot=<0 0 -0.0729 -0.9973>}
 toe_tip_bk_l {parent=toe_bk_l, loc=<4.4515 0 0>, rot=<0 0 -0.226 -0.9741>}
 ball_bk_lSocket {parent=ball_bk_l, type=Null, loc=<2.2878 0 0>, rot=<0 0 0 1>}
 tail_01 {parent=hips, loc=<-10.1036 -4.046 0>, rot=<-0.0242 -0.7067 0.7067 0.0242>}
 tail_02 {parent=tail_01, loc=<8.8431 0 0>, rot=<0 -0.0045 0 -1>}
 tail_03 {parent=tail_02, loc=<8.8438 0 0>, rot=<0 0.0315 0 -0.9995>}
 tail_04 {parent=tail_03, loc=<8.8346 0 0>, rot=<0 0.0277 0 -0.9996>}
 tail_04Socket {parent=tail_04, type=Null, loc=<7.1555 0 2.8692>, rot=<0 0 0 1>}
 thigh_bk_r {parent=hips, loc=<-3.6025 0.5607 6.2814>, rot=<0.7316 -0.6817 0 0>}
 knee_bk_r {parent=thigh_bk_r, loc=<-16.6492 -0.0001 0>, rot=<0 0 0.4212 -0.907>}
 ankle_bk_r {parent=knee_bk_r, loc=<-12.9954 0 0>, rot=<0 0 -0.4271 -0.9042>}
 ball_bk_r {parent=ankle_bk_r, loc=<-11.8273 0 0>, rot=<0 0 -0.5311 -0.8473>}
 toe_bk_r {parent=ball_bk_r, loc=<-2.1609 0 0>, rot=<0 0 -0.0729 -0.9973>}
 toe_tip_bk_r {parent=toe_bk_r, loc=<-4.4515 0 0>, rot=<0 0 -0.226 -0.9741>}
 ball_bk_rSocket {parent=ball_bk_r, type=Null, loc=<-2.4773 0 0>, rot=<0 0 0 1>}
 pelvisSocket {parent=pelvis, type=Null, loc=<0 -8.8042 0>, rot=<0 0 0 1>}
 IK_Foot_Rear_Root {parent=root, loc=<0 -30.8181 0>, rot=<0.0444 0 0 -0.999>}
 IK_Foot_Rear_L {parent=IK_Foot_Rear_Root, loc=<6.2809 -0.096 2.3511>, rot=<-0.3842 -0.5936 -0.3842 -0.5936>}
 IK_Foot_Rear_R {parent=IK_Foot_Rear_Root, loc=<-6.2809 -0.096 2.3511>, rot=<0.5936 -0.3842 0.5936 -0.3842>}
 IK_Foot_Front_Root {parent=root, loc=<0 3.5274 0>, rot=<0.047 0 -0.0001 -0.9989>}
 IK_Foot_Front_L {parent=IK_Foot_Front_Root, loc=<7.5169 9.5389 3.7395>, rot=<-0.3935 -0.5883 -0.3993 -0.5828>}
 IK_Foot_Front_R {parent=IK_Foot_Front_Root, loc=<-7.5141 9.5408 3.7399>, rot=<0.5827 -0.3994 0.5883 -0.3934>}
 attach_emote_03 {parent=root, loc=<0 0 0>, rot=<0 0 0 1>}
 attach_emote_02 {parent=root, loc=<0 0 0>, rot=<0 0 0 1>}
 attach_emote_01 {parent=root, loc=<0 0 0>, rot=<0 0 0 1>}
 attach {parent=root, loc=<0 0 0>, rot=<0 0 0 1>}
 rootSocket {parent=root, type=Null, loc=<0 0 0>, rot=<0 0 0 1>}
 BothLegIK {type=Curve, loc=<0 0 0>, rot=<0 0 0 1>}
 L_eye_look_up_pose {type=Curve, loc=<0 0 0>, rot=<0 0 0 1>}
 L_wide_pose {type=Curve, loc=<0 0 0>, rot=<0 0 0 1>}
 R_lip_corner_narrow_pose {type=Curve, loc=<0 0 0>, rot=<0 0 0 1>}
 R_upper_lip_lower_pose {type=Curve, loc=<0 0 0>, rot=<0 0 0 1>}
 L_lip_corner_narrow_pose {type=Curve, loc=<0 0 0>, rot=<0 0 0 1>}
 L_upper_lip_lower_pose {type=Curve, loc=<0 0 0>, rot=<0 0 0 1>}
 R_lower_lip_up_pose {type=Curve, loc=<0 0 0>, rot=<0 0 0 1>}
 R_brow_squeeze_in_pose {type=Curve, loc=<0 0 0>, rot=<0 0 0 1>}
 L_lip_corner_wide_pose {type=Curve, loc=<0 0 0>, rot=<0 0 0 1>}
 L_upperLidDown_pose {type=Curve, loc=<0 0 0>, rot=<0 0 0 1>}
 R_lower_lip_down_pose {type=Curve, loc=<0 0 0>, rot=<0 0 0 1>}
 R_upper_lip_raiser_pose {type=Curve, loc=<0 0 0>, rot=<0 0 0 1>}
 L_lower_lip_down_pose {type=Curve, loc=<0 0 0>, rot=<0 0 0 1>}
 L_upper_lip_raiser_pose {type=Curve, loc=<0 0 0>, rot=<0 0 0 1>}
 L_brow_squeeze_in_pose {type=Curve, loc=<0 0 0>, rot=<0 0 0 1>}
 L_lower_lip_up_pose {type=Curve, loc=<0 0 0>, rot=<0 0 0 1>}
 R_upperLidDown_pose {type=Curve, loc=<0 0 0>, rot=<0 0 0 1>}
 R_wide_pose {type=Curve, loc=<0 0 0>, rot=<0 0 0 1>}
 R_blink_pose {type=Curve, loc=<0 0 0>, rot=<0 0 0 1>}
 L_eye_look_left_pose {type=Curve, loc=<0 0 0>, rot=<0 0 0 1>}
 L_eye_look_down_pose {type=Curve, loc=<0 0 0>, rot=<0 0 0 1>}
 L_blink_pose {type=Curve, loc=<0 0 0>, rot=<0 0 0 1>}
 Reaction_Happy {type=Curve, loc=<0 0 0>, rot=<0 0 0 1>}
 L_eye_look_right_pose {type=Curve, loc=<0 0 0>, rot=<0 0 0 1>}
 R_lip_corner_wide_pose {type=Curve, loc=<0 0 0>, rot=<0 0 0 1>}
 L_smile_pose {type=Curve, loc=<0 0 0>, rot=<0 0 0 1>}
 R_smile_pose {type=Curve, loc=<0 0 0>, rot=<0 0 0 1>}
 L_frown_pose {type=Curve, loc=<0 0 0>, rot=<0 0 0 1>}
 R_frown_pose {type=Curve, loc=<0 0 0>, rot=<0 0 0 1>}
 C_nose_wrinkler_pose {type=Curve, loc=<0 0 0>, rot=<0 0 0 1>}
 L_cheek_up_pose {type=Curve, loc=<0 0 0>, rot=<0 0 0 1>}
 R_cheek_up_pose {type=Curve, loc=<0 0 0>, rot=<0 0 0 1>}
 jaw_open_pose {type=Curve, loc=<0 0 0>, rot=<0 0 0 1>}
 jaw_left_pose {type=Curve, loc=<0 0 0>, rot=<0 0 0 1>}
 jaw_right_pose {type=Curve, loc=<0 0 0>, rot=<0 0 0 1>}
 jaw_up_pose {type=Curve, loc=<0 0 0>, rot=<0 0 0 1>}
 jaw_forward_pose {type=Curve, loc=<0 0 0>, rot=<0 0 0 1>}
 jaw_back_pose {type=Curve, loc=<0 0 0>, rot=<0 0 0 1>}
 eye_convergence_pose {type=Curve, loc=<0 0 0>, rot=<0 0 0 1>}
 tongue_forward_pose {type=Curve, loc=<0 0 0>, rot=<0 0 0 1>}
 tongue_back_pose {type=Curve, loc=<0 0 0>, rot=<0 0 0 1>}
 tongue_up_pose {type=Curve, loc=<0 0 0>, rot=<0 0 0 1>}
 tongue_down_pose {type=Curve, loc=<0 0 0>, rot=<0 0 0 1>}
 L_brow_In_Up_pose {type=Curve, loc=<0 0 0>, rot=<0 0 0 1>}
 L_brow_Ot_Up_pose {type=Curve, loc=<0 0 0>, rot=<0 0 0 1>}
 L_brow_In_Dn_pose {type=Curve, loc=<0 0 0>, rot=<0 0 0 1>}
 L_brow_Ot_Dn_pose {type=Curve, loc=<0 0 0>, rot=<0 0 0 1>}
 R_brow_In_Up_pose {type=Curve, loc=<0 0 0>, rot=<0 0 0 1>}
 R_brow_Ot_Up_pose {type=Curve, loc=<0 0 0>, rot=<0 0 0 1>}
 R_brow_In_Dn_pose {type=Curve, loc=<0 0 0>, rot=<0 0 0 1>}
 R_brow_Ot_Dn_pose {type=Curve, loc=<0 0 0>, rot=<0 0 0 1>}
 upper_lip_up_pose {type=Curve, loc=<0 0 0>, rot=<0 0 0 1>}
 upper_lip_dn_pose {type=Curve, loc=<0 0 0>, rot=<0 0 0 1>}
 lower_lip_up_pose {type=Curve, loc=<0 0 0>, rot=<0 0 0 1>}
 lower_lip_dn_pose {type=Curve, loc=<0 0 0>, rot=<0 0 0 1>}
 L_smile_open_correction_pose {type=Curve, loc=<0 0 0>, rot=<0 0 0 1>}
 R_smile_open_correction_pose {type=Curve, loc=<0 0 0>, rot=<0 0 0 1>}
 tongue_left_pose {type=Curve, loc=<0 0 0>, rot=<0 0 0 1>}
 tongue_right_pose {type=Curve, loc=<0 0 0>, rot=<0 0 0 1>}
 reverse_jaw_pose {type=Curve, loc=<0 0 0>, rot=<0 0 0 1>}
 base_pose {type=Curve, loc=<0 0 0>, rot=<0 0 0 1>}
 R_eye_look_left_pose {type=Curve, loc=<0 0 0>, rot=<0 0 0 1>}
 R_eye_look_right_pose {type=Curve, loc=<0 0 0>, rot=<0 0 0 1>}
 R_eye_look_up_pose {type=Curve, loc=<0 0 0>, rot=<0 0 0 1>}
 R_eye_look_down_pose {type=Curve, loc=<0 0 0>, rot=<0 0 0 1>}
 L_lower_lid_up_pose {type=Curve, loc=<0 0 0>, rot=<0 0 0 1>}
 R_lower_lid_up_pose {type=Curve, loc=<0 0 0>, rot=<0 0 0 1>}
 tongue_curl_tip_up_pose {type=Curve, loc=<0 0 0>, rot=<0 0 0 1>}
 tongue_curl_tip_down_pose {type=Curve, loc=<0 0 0>, rot=<0 0 0 1>}
 tongue_curl_mid_up_pose {type=Curve, loc=<0 0 0>, rot=<0 0 0 1>}
 tongue_curl_mid_down_pose {type=Curve, loc=<0 0 0>, rot=<0 0 0 1>}
 L_tongue_twist_pose {type=Curve, loc=<0 0 0>, rot=<0 0 0 1>}
 R_tongue_twist_pose {type=Curve, loc=<0 0 0>, rot=<0 0 0 1>}
 AttachEmote1InPelvisSpace {type=Curve, loc=<0 0 0>, rot=<0 0 0 1>}
 AttachEmote2InPelvisSpace {type=Curve, loc=<0 0 0>, rot=<0 0 0 1>}
 AttachEmote3InPelvisSpace {type=Curve, loc=<0 0 0>, rot=<0 0 0 1>}
 AnimateAcc {type=Curve, loc=<0 0 0>, rot=<0 0 0 1>}
 pupil_shape {type=Curve, loc=<0 0 0>, rot=<0 0 0 1>}
 pupil_size {type=Curve, loc=<0 0 0>, rot=<0 0 0 1>}
 bypass_poseAsset {type=Curve, loc=<0 0 0>, rot=<0 0 0 1>}
 root/RootJoint {type=Connector, loc=<0 0 0>, rot=<0 0 0 1>}
 root/Body {type=Connector, loc=<0 0 0>, rot=<0 0 0 1>}
 Capsule/Root {type=Connector, loc=<0 0 0>, rot=<0 0 0 1>}
 Emote Prop 1/Start {type=Connector, loc=<0 0 0>, rot=<0 0 0 1>}
 Emote Prop 1/Parent {type=Connector, loc=<0 0 0>, rot=<0 0 0 1>}
 Emote Prop 1/End {type=Connector, loc=<0 0 0>, rot=<0 0 0 1>}
 Emote Prop 1/Orient Spaces {type=Connector, loc=<0 0 0>, rot=<0 0 0 1>}
 Emote Prop 2/Start {type=Connector, loc=<0 0 0>, rot=<0 0 0 1>}
 Emote Prop 2/Parent {type=Connector, loc=<0 0 0>, rot=<0 0 0 1>}
 Emote Prop 2/End {type=Connector, loc=<0 0 0>, rot=<0 0 0 1>}
 Emote Prop 2/Orient Spaces {type=Connector, loc=<0 0 0>, rot=<0 0 0 1>}
 Emote Prop 3/Start {type=Connector, loc=<0 0 0>, rot=<0 0 0 1>}
 Emote Prop 3/Parent {type=Connector, loc=<0 0 0>, rot=<0 0 0 1>}
 Emote Prop 3/End {type=Connector, loc=<0 0 0>, rot=<0 0 0 1>}
 Emote Prop 3/Orient Spaces {type=Connector, loc=<0 0 0>, rot=<0 0 0 1>}
 Body/Parent {type=Connector, loc=<0 0 0>, rot=<0 0 0 1>}
 Body/Body {type=Connector, loc=<0 0 0>, rot=<0 0 0 1>}
 Body/Right Hip {type=Connector, loc=<0 0 0>, rot=<0 0 0 1>}
 Body/Left Hip {type=Connector, loc=<0 0 0>, rot=<0 0 0 1>}
 Spine/SpineRoot {type=Connector, loc=<0 0 0>, rot=<0 0 0 1>}
 Spine/Spine 3 {type=Connector, loc=<0 0 0>, rot=<0 0 0 1>}
 Spine/Spine 1 {type=Connector, loc=<0 0 0>, rot=<0 0 0 1>}
 Spine/Parent {type=Connector, loc=<0 0 0>, rot=<0 0 0 1>}
 Spine/Spine 2 {type=Connector, loc=<0 0 0>, rot=<0 0 0 1>}
 Neck/Root {type=Connector, loc=<0 0 0>, rot=<0 0 0 1>}
 Neck/Head {type=Connector, loc=<0 0 0>, rot=<0 0 0 1>}
 Neck/NeckTip {type=Connector, loc=<0 0 0>, rot=<0 0 0 1>}
 Neck/Parent {type=Connector, loc=<0 0 0>, rot=<0 0 0 1>}
 Neck/NeckSpaces {type=Connector, loc=<0 0 0>, rot=<0 0 0 1>}
 Neck/HeadSpaces {type=Connector, loc=<0 0 0>, rot=<0 0 0 1>}
 Face/Head {type=Connector, loc=<0 0 0>, rot=<0 0 0 1>}
 Ear Root L/Start {type=Connector, loc=<0 0 0>, rot=<0 0 0 1>}
 Ear Root L/Parent {type=Connector, loc=<0 0 0>, rot=<0 0 0 1>}
 Ear Root L/End {type=Connector, loc=<0 0 0>, rot=<0 0 0 1>}
 Ear Root L/Orient Spaces {type=Connector, loc=<0 0 0>, rot=<0 0 0 1>}
 Ear L/Start {type=Connector, loc=<0 0 0>, rot=<0 0 0 1>}
 Ear L/Parent {type=Connector, loc=<0 0 0>, rot=<0 0 0 1>}
 Ear L/End {type=Connector, loc=<0 0 0>, rot=<0 0 0 1>}
 Ear L/Orient Spaces {type=Connector, loc=<0 0 0>, rot=<0 0 0 1>}
 Ear Micro L/Parent {type=Connector, loc=<0 0 0>, rot=<0 0 0 1>}
 Ear Micro L/Bones {type=Connector, loc=<0 0 0>, rot=<0 0 0 1>}
 Ear Micro L/Spaces {type=Connector, loc=<0 0 0>, rot=<0 0 0 1>}
 Ear Micro L/Orient Spaces {type=Connector, loc=<0 0 0>, rot=<0 0 0 1>}
 Ear Micro L/Override Parents {type=Connector, loc=<0 0 0>, rot=<0 0 0 1>}
 Ear Root R/Start {type=Connector, loc=<0 0 0>, rot=<0 0 0 1>}
 Ear Root R/Parent {type=Connector, loc=<0 0 0>, rot=<0 0 0 1>}
 Ear Root R/End {type=Connector, loc=<0 0 0>, rot=<0 0 0 1>}
 Ear Root R/Orient Spaces {type=Connector, loc=<0 0 0>, rot=<0 0 0 1>}
 Ear R/Start {type=Connector, loc=<0 0 0>, rot=<0 0 0 1>}
 Ear R/Parent {type=Connector, loc=<0 0 0>, rot=<0 0 0 1>}
 Ear R/End {type=Connector, loc=<0 0 0>, rot=<0 0 0 1>}
 Ear R/Orient Spaces {type=Connector, loc=<0 0 0>, rot=<0 0 0 1>}
 Ear Micro R/Parent {type=Connector, loc=<0 0 0>, rot=<0 0 0 1>}
 Ear Micro R/Bones {type=Connector, loc=<0 0 0>, rot=<0 0 0 1>}
 Ear Micro R/Spaces {type=Connector, loc=<0 0 0>, rot=<0 0 0 1>}
 Ear Micro R/Orient Spaces {type=Connector, loc=<0 0 0>, rot=<0 0 0 1>}
 Ear Micro R/Override Parents {type=Connector, loc=<0 0 0>, rot=<0 0 0 1>}
 Tail/HipRoot {type=Connector, loc=<0 0 0>, rot=<0 0 0 1>}
 Tail/Tail {type=Connector, loc=<0 0 0>, rot=<0 0 0 1>}
 Tail/Parent {type=Connector, loc=<0 0 0>, rot=<0 0 0 1>}
 Tail/Orient Spaces {type=Connector, loc=<0 0 0>, rot=<0 0 0 1>}
 Back Leg L/Start {type=Connector, loc=<0 0 0>, rot=<0 0 0 1>}
 Back Leg L/Mid {type=Connector, loc=<0 0 0>, rot=<0 0 0 1>}
 Back Leg L/End {type=Connector, loc=<0 0 0>, rot=<0 0 0 1>}
 Back Leg L/Ball {type=Connector, loc=<0 0 0>, rot=<0 0 0 1>}
 Back Leg L/Toe {type=Connector, loc=<0 0 0>, rot=<0 0 0 1>}
 Back Leg L/Toe Tip {type=Connector, loc=<0 0 0>, rot=<0 0 0 1>}
 Back Leg L/Parent {type=Connector, loc=<0 0 0>, rot=<0 0 0 1>}
 Back Leg L/IK Parent {type=Connector, loc=<0 0 0>, rot=<0 0 0 1>}
 Back Leg L/IK Spaces {type=Connector, loc=<0 0 0>, rot=<0 0 0 1>}
 Back Leg L/Scapula {type=Connector, loc=<0 0 0>, rot=<0 0 0 1>}
 Front Leg L/Start {type=Connector, loc=<0 0 0>, rot=<0 0 0 1>}
 Front Leg L/Mid {type=Connector, loc=<0 0 0>, rot=<0 0 0 1>}
 Front Leg L/End {type=Connector, loc=<0 0 0>, rot=<0 0 0 1>}
 Front Leg L/Ball {type=Connector, loc=<0 0 0>, rot=<0 0 0 1>}
 Front Leg L/Toe {type=Connector, loc=<0 0 0>, rot=<0 0 0 1>}
 Front Leg L/Toe Tip {type=Connector, loc=<0 0 0>, rot=<0 0 0 1>}
 Front Leg L/Parent {type=Connector, loc=<0 0 0>, rot=<0 0 0 1>}
 Front Leg L/IK Parent {type=Connector, loc=<0 0 0>, rot=<0 0 0 1>}
 Front Leg L/IK Spaces {type=Connector, loc=<0 0 0>, rot=<0 0 0 1>}
 Front Leg L/Scapula {type=Connector, loc=<0 0 0>, rot=<0 0 0 1>}
 Back Leg R/Start {type=Connector, loc=<0 0 0>, rot=<0 0 0 1>}
 Back Leg R/Mid {type=Connector, loc=<0 0 0>, rot=<0 0 0 1>}
 Back Leg R/End {type=Connector, loc=<0 0 0>, rot=<0 0 0 1>}
 Back Leg R/Ball {type=Connector, loc=<0 0 0>, rot=<0 0 0 1>}
 Back Leg R/Toe {type=Connector, loc=<0 0 0>, rot=<0 0 0 1>}
 Back Leg R/Toe Tip {type=Connector, loc=<0 0 0>, rot=<0 0 0 1>}
 Back Leg R/Parent {type=Connector, loc=<0 0 0>, rot=<0 0 0 1>}
 Back Leg R/IK Parent {type=Connector, loc=<0 0 0>, rot=<0 0 0 1>}
 Back Leg R/IK Spaces {type=Connector, loc=<0 0 0>, rot=<0 0 0 1>}
 Back Leg R/Scapula {type=Connector, loc=<0 0 0>, rot=<0 0 0 1>}
 Front Leg R/Start {type=Connector, loc=<0 0 0>, rot=<0 0 0 1>}
 Front Leg R/Mid {type=Connector, loc=<0 0 0>, rot=<0 0 0 1>}
 Front Leg R/End {type=Connector, loc=<0 0 0>, rot=<0 0 0 1>}
 Front Leg R/Ball {type=Connector, loc=<0 0 0>, rot=<0 0 0 1>}
 Front Leg R/Toe {type=Connector, loc=<0 0 0>, rot=<0 0 0 1>}
 Front Leg R/Toe Tip {type=Connector, loc=<0 0 0>, rot=<0 0 0 1>}
 Front Leg R/Parent {type=Connector, loc=<0 0 0>, rot=<0 0 0 1>}
 Front Leg R/IK Parent {type=Connector, loc=<0 0 0>, rot=<0 0 0 1>}
 Front Leg R/IK Spaces {type=Connector, loc=<0 0 0>, rot=<0 0 0 1>}
 Front Leg R/Scapula {type=Connector, loc=<0 0 0>, rot=<0 0 0 1>}
 Pin IK Bones/Root {type=Connector, loc=<0 0 0>, rot=<0 0 0 1>}
 Pin IK Bones/Drivers {type=Connector, loc=<0 0 0>, rot=<0 0 0 1>}
 Pin IK Bones/Driven {type=Connector, loc=<0 0 0>, rot=<0 0 0 1>}
]
