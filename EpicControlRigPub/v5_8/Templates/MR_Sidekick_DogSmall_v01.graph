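# Modular Control Rig template for a small sidekick dog, with a quadruped skeleton, detailed face, ears, tail and emote attach bones.
# Named connectors expose module slots for root, body, spine, neck, face, ears, tail, four legs and three emote props.
# Leg modules offer IK spaces and scapula connectors, and a Pin IK Bones module has driver and driven connectors.
# The event graph holds only an empty BeginExecution, alongside facial pose and body curves.
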ControlRigBlueprint /EpicControlRigPub/v5_8/Templates/MR_Sidekick_DogSmall_v01
graphs [
 RigVMFunctionLibrary {links [], nodes []}
 RigVMModel {nodes [RigUnit_BeginExecution = &RigUnit_BeginExecution()]}
]
hierarchy [
 BothLegIK {type=Curve, loc=<0 0 0>, rot=<0 0 0 1>}
 L_eye_look_up_pose {type=Curve, loc=<0 0 0>, rot=<0 0 0 1>}
 L_wide_pose {type=Curve, loc=<0 0 0>, rot=<0 0 0 1>}
 R_lip_corner_narrow_pose {type=Curve, loc=<0 0 0>, rot=<0 0 0 1>}
 R_upper_lip_lower_pose {type=Curve, loc=<0 0 0>, rot=<0 0 0 1>}
 L_lip_corner_narrow_pose {type=Curve, loc=<0 0 0>, rot=<0 0 0 1>}
 L_upper_lip_lower_pose {type=Curve, loc=<0 0 0>, rot=<0 0 0 1>}
 R_lower_lip_up_pose {type=Curve, loc=<0 0 0>, rot=<0 0 0 1>}
 R_brow_squeeze_in_pose {type=Curve, loc=<0 0 0>, rot=<0 0 0 1>}
 L_lip_corner_wide_pose {type=Curve, loc=<0 0 0>, rot=<0 0 0 1>}
 L_upperLidDown_pose {type=Curve, loc=<0 0 0>, rot=<0 0 0 1>}
 R_lower_lip_down_pose {type=Curve, loc=<0 0 0>, rot=<0 0 0 1>}
 R_upper_lip_raiser_pose {type=Curve, loc=<0 0 0>, rot=<0 0 0 1>}
 L_lower_lip_down_pose {type=Curve, loc=<0 0 0>, rot=<0 0 0 1>}
 L_upper_lip_raiser_pose {type=Curve, loc=<0 0 0>, rot=<0 0 0 1>}
 L_brow_squeeze_in_pose {type=Curve, loc=<0 0 0>, rot=<0 0 0 1>}
 L_lower_lip_up_pose {type=Curve, loc=<0 0 0>, rot=<0 0 0 1>}
 R_upperLidDown_pose {type=Curve, loc=<0 0 0>, rot=<0 0 0 1>}
 R_wide_pose {type=Curve, loc=<0 0 0>, rot=<0 0 0 1>}
 R_blink_pose {type=Curve, loc=<0 0 0>, rot=<0 0 0 1>}
 L_eye_look_left_pose {type=Curve, loc=<0 0 0>, rot=<0 0 0 1>}
 L_eye_look_down_pose {type=Curve, loc=<0 0 0>, rot=<0 0 0 1>}
 L_blink_pose {type=Curve, loc=<0 0 0>, rot=<0 0 0 1>}
 Reaction_Happy {type=Curve, loc=<0 0 0>, rot=<0 0 0 1>}
 L_eye_look_right_pose {type=Curve, loc=<0 0 0>, rot=<0 0 0 1>}
 R_lip_corner_wide_pose {type=Curve, loc=<0 0 0>, rot=<0 0 0 1>}
 L_smile_pose {type=Curve, loc=<0 0 0>, rot=<0 0 0 1>}
 R_smile_pose {type=Curve, loc=<0 0 0>, rot=<0 0 0 1>}
 L_frown_pose {type=Curve, loc=<0 0 0>, rot=<0 0 0 1>}
 R_frown_pose {type=Curve, loc=<0 0 0>, rot=<0 0 0 1>}
 C_nose_wrinkler_pose {type=Curve, loc=<0 0 0>, rot=<0 0 0 1>}
 L_cheek_up_pose {type=Curve, loc=<0 0 0>, rot=<0 0 0 1>}
 R_cheek_up_pose {type=Curve, loc=<0 0 0>, rot=<0 0 0 1>}
 jaw_open_pose {type=Curve, loc=<0 0 0>, rot=<0 0 0 1>}
 jaw_left_pose {type=Curve, loc=<0 0 0>, rot=<0 0 0 1>}
 jaw_right_pose {type=Curve, loc=<0 0 0>, rot=<0 0 0 1>}
 jaw_up_pose {type=Curve, loc=<0 0 0>, rot=<0 0 0 1>}
 jaw_forward_pose {type=Curve, loc=<0 0 0>, rot=<0 0 0 1>}
 jaw_back_pose {type=Curve, loc=<0 0 0>, rot=<0 0 0 1>}
 eye_convergence_pose {type=Curve, loc=<0 0 0>, rot=<0 0 0 1>}
 tongue_forward_pose {type=Curve, loc=<0 0 0>, rot=<0 0 0 1>}
 tongue_back_pose {type=Curve, loc=<0 0 0>, rot=<0 0 0 1>}
 tongue_up_pose {type=Curve, loc=<0 0 0>, rot=<0 0 0 1>}
 tongue_down_pose {type=Curve, loc=<0 0 0>, rot=<0 0 0 1>}
 L_brow_In_Up_pose {type=Curve, loc=<0 0 0>, rot=<0 0 0 1>}
 L_brow_Ot_Up_pose {type=Curve, loc=<0 0 0>, rot=<0 0 0 1>}
 L_brow_In_Dn_pose {type=Curve, loc=<0 0 0>, rot=<0 0 0 1>}
 L_brow_Ot_Dn_pose {type=Curve, loc=<0 0 0>, rot=<0 0 0 1>}
 R_brow_In_Up_pose {type=Curve, loc=<0 0 0>, rot=<0 0 0 1>}
 R_brow_Ot_Up_pose {type=Curve, loc=<0 0 0>, rot=<0 0 0 1>}
 R_brow_In_Dn_pose {type=Curve, loc=<0 0 0>, rot=<0 0 0 1>}
 R_brow_Ot_Dn_pose {type=Curve, loc=<0 0 0>, rot=<0 0 0 1>}
 upper_lip_up_pose {type=Curve, loc=<0 0 0>, rot=<0 0 0 1>}
 upper_lip_dn_pose {type=Curve, loc=<0 0 0>, rot=<0 0 0 1>}
 lower_lip_up_pose {type=Curve, loc=<0 0 0>, rot=<0 0 0 1>}
 lower_lip_dn_pose {type=Curve, loc=<0 0 0>, rot=<0 0 0 1>}
 L_smile_open_correction_pose {type=Curve, loc=<0 0 0>, rot=<0 0 0 1>}
 R_smile_open_correction_pose {type=Curve, loc=<0 0 0>, rot=<0 0 0 1>}
 tongue_left_pose {type=Curve, loc=<0 0 0>, rot=<0 0 0 1>}
 tongue_right_pose {type=Curve, loc=<0 0 0>, rot=<0 0 0 1>}
 reverse_jaw_pose {type=Curve, loc=<0 0 0>, rot=<0 0 0 1>}
 base_pose {type=Curve, loc=<0 0 0>, rot=<0 0 0 1>}
 R_eye_look_left_pose {type=Curve, loc=<0 0 0>, rot=<0 0 0 1>}
 R_eye_look_right_pose {type=Curve, loc=<0 0 0>, rot=<0 0 0 1>}
 R_eye_look_up_pose {type=Curve, loc=<0 0 0>, rot=<0 0 0 1>}
 R_eye_look_down_pose {type=Curve, loc=<0 0 0>, rot=<0 0 0 1>}
 L_lower_lid_up_pose {type=Curve, loc=<0 0 0>, rot=<0 0 0 1>}
 R_lower_lid_up_pose {type=Curve, loc=<0 0 0>, rot=<0 0 0 1>}
 tongue_curl_tip_up_pose {type=Curve, loc=<0 0 0>, rot=<0 0 0 1>}
 tongue_curl_tip_down_pose {type=Curve, loc=<0 0 0>, rot=<0 0 0 1>}
 tongue_curl_mid_up_pose {type=Curve, loc=<0 0 0>, rot=<0 0 0 1>}
 tongue_curl_mid_down_pose {type=Curve, loc=<0 0 0>, rot=<0 0 0 1>}
 L_tongue_twist_pose {type=Curve, loc=<0 0 0>, rot=<0 0 0 1>}
 R_tongue_twist_pose {type=Curve, loc=<0 0 0>, rot=<0 0 0 1>}
 AttachEmote1InPelvisSpace {type=Curve, loc=<0 0 0>, rot=<0 0 0 1>}
 AttachEmote2InPelvisSpace {type=Curve, loc=<0 0 0>, rot=<0 0 0 1>}
 AttachEmote3InPelvisSpace {type=Curve, loc=<0 0 0>, rot=<0 0 0 1>}
 AnimateAcc {type=Curve, loc=<0 0 0>, rot=<0 0 0 1>}
 pupil_shape {type=Curve, loc=<0 0 0>, rot=<0 0 0 1>}
 pupil_size {type=Curve, loc=<0 0 0>, rot=<0 0 0 1>}
 bypass_poseAsset {type=Curve, loc=<0 0 0>, rot=<0 0 0 1>}
 Body_Scale {type=Curve, loc=<0 0 0>, rot=<0 0 0 1>}
 BodyMorph_LOD0 {type=Curve, loc=<0 0 0>, rot=<0 0 0 1>}
 HoodieMorph_LOD0 {type=Curve, loc=<0 0 0>, rot=<0 0 0 1>}
 BodyScale {type=Curve, loc=<0 0 0>, rot=<0 0 0 1>}
 root {loc=<0 0 0>, rot=<0 0 0 1>}
 pelvis {parent=root, loc=<0 -9.1937 28.2113>, rot=<0.5131 0.4865 -0.5131 -0.4865>}
 spine_01 {parent=pelvis, loc=<0 0 0>, rot=<0 0 -0.0124 -0.9999>}
 spine_02 {parent=spine_01, loc=<11.0133 -0.2742 0>, rot=<0 0 0.014 -0.9999>}
 spine_03 {parent=spine_02, loc=<9.8251 0 0>, rot=<0 0 0.031 -0.9995>}
 neck_01 {parent=spine_03, loc=<0.7833 0.3499 0>, rot=<0 0 0.5645 -0.8254>}
 neck_02 {parent=neck_01, loc=<7.5647 0 0>, rot=<0 0 -0.0522 -0.9986>}
 head {parent=neck_02, loc=<6.1014 0 0>, rot=<-0.1539 -0.6902 0.1539 -0.6902>}
 face_attach {parent=head, loc=<0 0 0>, rot=<0 0 0 1>}
 jaw {parent=face_attach, loc=<0 1.9538 4.8461>, rot=<0 0.7071 0 -0.7071>}
 lip_lower_03_l {parent=jaw, loc=<-5.0785 8.1298 -3.1007>, rot=<-0.13 -0.7302 -0.0862 -0.6652>}
 lip_lower_02_l {parent=jaw, loc=<-5.0079 9.9522 -2.6356>, rot=<-0.1052 -0.72 -0.1123 -0.6767>}
 lip_lower_01_l {parent=jaw, loc=<-4.8924 11.9342 -1.6763>, rot=<-0.2618 -0.676 -0.2463 -0.6433>}
 lip_lower_mid {parent=jaw, loc=<-4.9138 12.5583 0>, rot=<0.4939 0.506 0.506 0.4939>}
 lip_lower_01_r {parent=jaw, loc=<-4.8924 11.9343 1.6763>, rot=<0.676 -0.2618 -0.6433 0.2463>}
 lip_lower_02_r {parent=jaw, loc=<-5.0079 9.9523 2.6356>, rot=<0.72 -0.1052 -0.6767 0.1123>}
 lip_lower_03_r {parent=jaw, loc=<-5.0785 8.1298 3.1007>, rot=<0.7302 -0.13 -0.6652 0.0862>}
 teeth_lower {parent=jaw, loc=<-6.1384 7.8772 0>, rot=<0 -0.7071 0 -0.7071>}
 tongue_01 {parent=jaw, loc=<-3.6125 0.3959 -0.2739>, rot=<0.4938 -0.5061 0.5061 -0.4938>}
 tongue_02 {parent=tongue_01, loc=<-4.1659 0 0>, rot=<0 0.0745 0 -0.9972>}
 tongue_03 {parent=tongue_02, loc=<-4.2733 0 0>, rot=<0 -0.0397 0 -0.9992>}
 lip_corner_l {parent=face_attach, loc=<3.832 8.9145 0.0222>, rot=<0.0937 -0.0534 0.5415 -0.8337>}
 cheek_bt_l {parent=face_attach, loc=<5.6275 5.2127 -1.9259>, rot=<0.0056 -0.2755 -0.2636 -0.9245>}
 cheek_bt_r {parent=face_attach, loc=<-5.6275 5.2126 -1.9258>, rot=<0.9245 -0.2636 0.2755 0.0056>}
 reverse_jaw {parent=face_attach, loc=<0 0.0433 0.7402>, rot=<0 0.7071 0 -0.7071>}
 lip_upper_mid {parent=reverse_jaw, loc=<0.2686 15.1318 0>, rot=<-0.4824 -0.517 -0.517 -0.4824>}
 lip_upper_01_l {parent=reverse_jaw, loc=<-0.0345 14.2289 -2.0705>, rot=<-0.2245 -0.6851 -0.2091 -0.6606>}
 lip_upper_02_l {parent=reverse_jaw, loc=<-0.1136 12.1388 -3.1728>, rot=<-0.096 -0.7168 -0.1084 -0.6821>}
 lip_upper_03_l {parent=reverse_jaw, loc=<-0.2622 10.1801 -3.7875>, rot=<-0.0975 -0.688 -0.0698 -0.7158>}
 cheek_tp_in_l {parent=reverse_jaw, loc=<3.8126 9.6563 -3.8183>, rot=<-0.1785 -0.3811 -0.2158 -0.8811>}
 cheek_tp_ot_l {parent=reverse_jaw, loc=<4.2024 4.7989 -7.8743>, rot=<-0.0767 -0.5752 -0.2373 -0.7791>}
 cheek_tp_in_r {parent=reverse_jaw, loc=<3.8125 9.6563 3.8183>, rot=<-0.3811 0.1785 0.8811 -0.2158>}
 cheek_tp_ot_r {parent=reverse_jaw, loc=<4.2023 4.7989 7.8743>, rot=<-0.5752 0.0767 0.7791 -0.2373>}
 brow_ot_r {parent=reverse_jaw, loc=<8.4489 6.0452 6.4781>, rot=<0.6138 0.2707 -0.5866 -0.4537>}
 brow_md_r {parent=reverse_jaw, loc=<9.9946 7.7462 4.2458>, rot=<0.7283 0.0844 -0.6173 -0.2853>}
 brow_ct {parent=reverse_jaw, loc=<10.2183 9.1581 0>, rot=<-0.1324 -0.6946 0.1324 -0.6946>}
 brow_md_l {parent=reverse_jaw, loc=<9.9946 7.7462 -4.2458>, rot=<0.0844 -0.7283 0.2853 -0.6173>}
 brow_in_l {parent=reverse_jaw, loc=<9.9575 8.8512 -2.2993>, rot=<0.0835 -0.6941 0.1686 -0.6948>}
 brow_ot_l {parent=reverse_jaw, loc=<8.4489 6.0452 -6.4781>, rot=<0.2707 -0.6138 0.4537 -0.5866>}
 brow_in_r {parent=reverse_jaw, loc=<9.9575 8.8512 2.2993>, rot=<-0.6941 -0.0835 0.6948 0.1686>}
 eye_r {parent=reverse_jaw, loc=<6.8214 6.1614 3.1086>, rot=<0.7071 -0.0088 -0.7071 -0.0088>}
 lid_bt_r {parent=reverse_jaw, loc=<6.8214 6.1614 3.1086>, rot=<-0.7016 -0.003 0.7126 -0.0029>}
 lid_tp_r {parent=reverse_jaw, loc=<6.8214 6.1614 3.1086>, rot=<-0.7016 -0.003 0.7126 -0.0029>}
 nose_mid {parent=reverse_jaw, loc=<2.9239 15.1205 0>, rot=<-0.5471 -0.448 -0.448 -0.5471>}
 nostril_l {parent=nose_mid, loc=<0.5813 -0.7277 -0.5254>, rot=<0 0 0 -1>}
 nostril_r {parent=nose_mid, loc=<0.5813 0.7277 -0.5254>, rot=<0 -1 0 0>}
 lid_tp_l {parent=reverse_jaw, loc=<6.8214 6.1614 -3.1086>, rot=<0.003 -0.7016 -0.0029 -0.7126>}
 lid_bt_l {parent=reverse_jaw, loc=<6.8214 6.1614 -3.1086>, rot=<0.003 -0.7016 -0.0029 -0.7126>}
 eye_l {parent=reverse_jaw, loc=<6.8214 6.1614 -3.1086>, rot=<-0.0088 -0.7071 0.0088 -0.7071>}
 lip_upper_01_r {parent=reverse_jaw, loc=<-0.0346 14.229 2.0706>, rot=<0.6851 -0.2245 -0.6606 0.2091>}
 lip_upper_02_r {parent=reverse_jaw, loc=<-0.1137 12.1388 3.1728>, rot=<0.7168 -0.096 -0.6821 0.1084>}
 lip_upper_03_r {parent=reverse_jaw, loc=<-0.2622 10.1801 3.7875>, rot=<-0.688 0.0975 0.7158 -0.0698>}
 teeth_upper {parent=reverse_jaw, loc=<1.7757 10.1681 0>, rot=<0 -0.7071 0 -0.7071>}
 ear_01_l {parent=reverse_jaw, loc=<3.898 1.4723 -1.2636>, rot=<-0.0198 -0.1832 0.2102 -0.9601>}
 ear_02_l {parent=ear_01_l, loc=<7.8571 0 0>, rot=<0.0567 -0.1023 -0.1736 -0.9779>}
 ear_03_l {parent=ear_02_l, loc=<2.978 0 0>, rot=<0.0839 0.0547 -0.1614 -0.9818>}
 ear_04_l {parent=ear_03_l, loc=<2.867 0 0>, rot=<-0.0162 0.0844 0.0226 -0.996>}
 ear_01_r {parent=reverse_jaw, loc=<3.898 1.4722 1.2636>, rot=<-0.1832 0.0198 0.9601 0.2102>}
 ear_02_r {parent=ear_01_r, loc=<-7.8571 0 0>, rot=<0.0567 -0.1023 -0.1736 -0.9779>}
 ear_03_r {parent=ear_02_r, loc=<-2.9779 0 0>, rot=<0.0839 0.0547 -0.1614 -0.9818>}
 ear_04_r {parent=ear_03_r, loc=<-2.867 -0.0001 0>, rot=<-0.0162 0.0844 0.0226 -0.996>}
 lip_corner_r {parent=face_attach, loc=<-3.832 8.9145 0.0222>, rot=<0.8337 0.5415 0.0534 0.0937>}
 knot {parent=neck_02, loc=<0.4365 -10.2604 0>, rot=<0 0 0.2297 -0.9733>}
 knot_l_01 {parent=knot, loc=<0.4956 -0.4462 -1.0455>, rot=<-0.5002 -0.0487 0.0838 -0.8605>}
 knot_l_02 {parent=knot_l_01, loc=<0 -3.4249 0>, rot=<0.3093 -0.1922 0.104 -0.9255>}
 knot_r_01 {parent=knot, loc=<0.3052 -0.3478 1.1937>, rot=<0.0961 0.5746 0.8016 -0.1341>}
 knot_r_02 {parent=knot_r_01, loc=<0 3.1479 0>, rot=<0.2401 -0.3057 -0.1869 -0.9022>}
 tag {parent=neck_02, loc=<0.6431 11.1635 0>, rot=<0 0 0.2297 -0.9733>}
 shoulder_fr_l {parent=spine_03, loc=<2.147 1.488 -6.748>, rot=<-0.0045 0.0076 -0.7951 -0.6064>}
 elbow_fr_l {parent=shoulder_fr_l, loc=<11.9351 0 0>, rot=<0.0041 0.0009 0.0951 -0.9955>}
 wrist_fr_l {parent=elbow_fr_l, loc=<8.5313 0 0>, rot=<0.9818 -0.1889 0.0185 -0.0005>}
 ball_fr_l {parent=wrist_fr_l, loc=<3.3 0 0>, rot=<-0.0537 0.0764 -0.3272 -0.9403>}
 toe_fr_l {parent=ball_fr_l, loc=<1.4732 0 0>, rot=<-0.0077 -0.039 -0.1557 -0.987>}
 toe_tip_fr_l {parent=toe_fr_l, loc=<3.5769 0 0>, rot=<0.003 -0.0026 -0.121 -0.9926>}
 scapula_fr_l {parent=shoulder_fr_l, loc=<-5.2648 4.2357 0.0731>, rot=<-0.0004 -0.0056 0.3323 -0.9432>}
 shoulder_fr_r {parent=spine_03, loc=<2.147 1.488 6.748>, rot=<0.0076 0.0045 0.6064 -0.7951>}
 elbow_fr_r {parent=shoulder_fr_r, loc=<-11.9351 0 -0.0001>, rot=<0.0041 0.0009 0.0951 -0.9955>}
 wrist_fr_r {parent=elbow_fr_r, loc=<-8.5313 0 0>, rot=<0.9818 -0.1889 0.0185 -0.0005>}
 ball_fr_r {parent=wrist_fr_r, loc=<-3.3 0 0>, rot=<-0.0537 0.0764 -0.3272 -0.9403>}
 toe_fr_r {parent=ball_fr_r, loc=<-1.4732 0 0>, rot=<-0.0077 -0.039 -0.1557 -0.987>}
 toe_tip_fr_r {parent=toe_fr_r, loc=<-3.5768 0 0.0001>, rot=<0.003 -0.0026 -0.121 -0.9926>}
 scapula_fr_r {parent=shoulder_fr_r, loc=<5.2648 -4.2358 -0.0731>, rot=<-0.0004 -0.0056 0.3323 -0.9432>}
 suriken_l {parent=spine_02, loc=<3.1119 -5.201 -9.9223>, rot=<-0.335 -0.4608 0.647 -0.5068>}
 suriken_r {parent=spine_02, loc=<3.1119 -5.201 9.9223>, rot=<0.4608 -0.335 -0.5068 -0.647>}
 hips {parent=pelvis, loc=<0 0 0>, rot=<0 0 0 1>}
 thigh_bk_l {parent=hips, loc=<-3.1481 -1.1439 -4.4054>, rot=<0.7145 0.6991 -0.0205 0.0189>}
 knee_bk_l {parent=thigh_bk_l, loc=<12.9068 0 0>, rot=<0.0191 -0.009 0.3401 -0.9402>}
 ankle_bk_l {parent=knee_bk_l, loc=<9.5129 0 0>, rot=<-0.0174 0.0083 -0.3824 -0.9238>}
 ball_bk_l {parent=ankle_bk_l, loc=<6.9302 0 0>, rot=<-0.0127 -0.0228 -0.4083 -0.9125>}
 toe_bk_l {parent=ball_bk_l, loc=<1.8166 0 0>, rot=<0 0 -0.1426 -0.9898>}
 toe_tip_bk_l {parent=toe_bk_l, loc=<3.4203 0 0>, rot=<0 0 -0.2352 -0.9719>}
 thigh_bk_r {parent=hips, loc=<-3.148 -1.1439 4.4054>, rot=<0.6991 -0.7145 -0.0189 -0.0205>}
 knee_bk_r {parent=thigh_bk_r, loc=<-12.9068 0 0>, rot=<0.0191 -0.009 0.3401 -0.9402>}
 ankle_bk_r {parent=knee_bk_r, loc=<-9.5129 0 0>, rot=<-0.0174 0.0083 -0.3824 -0.9238>}
 ball_bk_r {parent=ankle_bk_r, loc=<-6.9302 0 0>, rot=<-0.0127 -0.0228 -0.4083 -0.9125>}
 toe_bk_r {parent=ball_bk_r, loc=<-1.8166 0 0>, rot=<0 0 -0.1426 -0.9898>}
 toe_tip_bk_r {parent=toe_bk_r, loc=<-3.4203 0 0>, rot=<0 0 -0.2352 -0.9719>}
 tail_01 {parent=hips, loc=<-7.0345 -3.9289 0.3513>, rot=<-0.0155 -0.7069 0.7069 0.0155>}
 tail_02 {parent=tail_01, loc=<5.329 0 0>, rot=<0 -0.0328 0 -0.9995>}
 tail_03 {parent=tail_02, loc=<5.3257 0 0>, rot=<0 -0.0045 0 -1>}
 tail_04 {parent=tail_03, loc=<5.3257 0 0>, rot=<0 0.0502 0 -0.9987>}
 tail_05 {parent=tail_04, loc=<5.3285 0 0>, rot=<0 -0.006 0 -1>}
 IK_Foot_Rear_Root {parent=root, loc=<0 -19.464 0>, rot=<0 -0.7824 0.6227 0>}
 IK_Foot_Rear_L {parent=IK_Foot_Rear_Root, loc=<-5.8444 -2.3606 -1.446>, rot=<0.6373 0.3064 -0.6373 -0.3064>}
 IK_Foot_Rear_R {parent=IK_Foot_Rear_Root, loc=<5.8444 -2.3606 -1.446>, rot=<-0.3064 0.6373 0.3064 -0.6373>}
 IK_Foot_Front_Root {parent=root, loc=<0 1.5578 0>, rot=<0.105 0 0 -0.9945>}
 IK_Foot_Front_L {parent=IK_Foot_Front_Root, loc=<6.1906 8.9391 4.4477>, rot=<-0.4489 -0.5618 -0.3412 -0.6054>}
 IK_Foot_Front_R {parent=IK_Foot_Front_Root, loc=<-6.1906 8.9391 4.4477>, rot=<0.6054 -0.3412 0.5618 -0.4489>}
 attach_emote_01 {parent=root, loc=<0 0 0>, rot=<0 0 0 1>}
 attach_emote_02 {parent=root, loc=<0 0 0>, rot=<0 0 0 1>}
 attach_emote_03 {parent=root, loc=<0 0 0>, rot=<0 0 0 1>}
 root/RootJoint {type=Connector, loc=<0 0 0>, rot=<0 0 0 1>}
 root/Body {type=Connector, loc=<0 0 0>, rot=<0 0 0 1>}
 Capsule/Root {type=Connector, loc=<0 0 0>, rot=<0 0 0 1>}
 Emote Prop 1/Start {type=Connector, loc=<0 0 0>, rot=<0 0 0 1>}
 Emote Prop 1/Parent {type=Connector, loc=<0 0 0>, rot=<0 0 0 1>}
 Emote Prop 1/End {type=Connector, loc=<0 0 0>, rot=<0 0 0 1>}
 Emote Prop 1/Orient Spaces {type=Connector, loc=<0 0 0>, rot=<0 0 0 1>}
 Emote Prop 2/Start {type=Connector, loc=<0 0 0>, rot=<0 0 0 1>}
 Emote Prop 2/Parent {type=Connector, loc=<0 0 0>, rot=<0 0 0 1>}
 Emote Prop 2/End {type=Connector, loc=<0 0 0>, rot=<0 0 0 1>}
 Emote Prop 2/Orient Spaces {type=Connector, loc=<0 0 0>, rot=<0 0 0 1>}
 Emote Prop 3/Start {type=Connector, loc=<0 0 0>, rot=<0 0 0 1>}
 Emote Prop 3/Parent {type=Connector, loc=<0 0 0>, rot=<0 0 0 1>}
 Emote Prop 3/End {type=Connector, loc=<0 0 0>, rot=<0 0 0 1>}
 Emote Prop 3/Orient Spaces {type=Connector, loc=<0 0 0>, rot=<0 0 0 1>}
 Body/Parent {type=Connector, loc=<0 0 0>, rot=<0 0 0 1>}
 Body/Body {type=Connector, loc=<0 0 0>, rot=<0 0 0 1>}
 Body/Right Hip {type=Connector, loc=<0 0 0>, rot=<0 0 0 1>}
 Body/Left Hip {type=Connector, loc=<0 0 0>, rot=<0 0 0 1>}
 Spine/SpineRoot {type=Connector, loc=<0 0 0>, rot=<0 0 0 1>}
 Spine/Spine 3 {type=Connector, loc=<0 0 0>, rot=<0 0 0 1>}
 Spine/Spine 1 {type=Connector, loc=<0 0 0>, rot=<0 0 0 1>}
 Spine/Parent {type=Connector, loc=<0 0 0>, rot=<0 0 0 1>}
 Spine/Spine 2 {type=Connector, loc=<0 0 0>, rot=<0 0 0 1>}
 Neck/Root {type=Connector, loc=<0 0 0>, rot=<0 0 0 1>}
 Neck/Head {type=Connector, loc=<0 0 0>, rot=<0 0 0 1>}
 Neck/NeckTip {type=Connector, loc=<0 0 0>, rot=<0 0 0 1>}
 Neck/Parent {type=Connector, loc=<0 0 0>, rot=<0 0 0 1>}
 Neck/NeckSpaces {type=Connector, loc=<0 0 0>, rot=<0 0 0 1>}
 Neck/HeadSpaces {type=Connector, loc=<0 0 0>, rot=<0 0 0 1>}
 Face/Head {type=Connector, loc=<0 0 0>, rot=<0 0 0 1>}
 Ear Root L/Start {type=Connector, loc=<0 0 0>, rot=<0 0 0 1>}
 Ear Root L/Parent {type=Connector, loc=<0 0 0>, rot=<0 0 0 1>}
 Ear Root L/End {type=Connector, loc=<0 0 0>, rot=<0 0 0 1>}
 Ear Root L/Orient Spaces {type=Connector, loc=<0 0 0>, rot=<0 0 0 1>}
 Ear L/Start {type=Connector, loc=<0 0 0>, rot=<0 0 0 1>}
 Ear L/Parent {type=Connector, loc=<0 0 0>, rot=<0 0 0 1>}
 Ear L/End {type=Connector, loc=<0 0 0>, rot=<0 0 0 1>}
 Ear L/Orient Spaces {type=Connector, loc=<0 0 0>, rot=<0 0 0 1>}
 Ear Root R/Start {type=Connector, loc=<0 0 0>, rot=<0 0 0 1>}
 Ear Root R/Parent {type=Connector, loc=<0 0 0>, rot=<0 0 0 1>}
 Ear Root R/End {type=Connector, loc=<0 0 0>, rot=<0 0 0 1>}
 Ear Root R/Orient Spaces {type=Connector, loc=<0 0 0>, rot=<0 0 0 1>}
 Ear R/Start {type=Connector, loc=<0 0 0>, rot=<0 0 0 1>}
 Ear R/Parent {type=Connector, loc=<0 0 0>, rot=<0 0 0 1>}
 Ear R/End {type=Connector, loc=<0 0 0>, rot=<0 0 0 1>}
 Ear R/Orient Spaces {type=Connector, loc=<0 0 0>, rot=<0 0 0 1>}
 Tail/HipRoot {type=Connector, loc=<0 0 0>, rot=<0 0 0 1>}
 Tail/Tail {type=Connector, loc=<0 0 0>, rot=<0 0 0 1>}
 Tail/Parent {type=Connector, loc=<0 0 0>, rot=<0 0 0 1>}
 Tail/Orient Spaces {type=Connector, loc=<0 0 0>, rot=<0 0 0 1>}
 Back Leg L/Start {type=Connector, loc=<0 0 0>, rot=<0 0 0 1>}
 Back Leg L/Mid {type=Connector, loc=<0 0 0>, rot=<0 0 0 1>}
 Back Leg L/End {type=Connector, loc=<0 0 0>, rot=<0 0 0 1>}
 Back Leg L/Ball {type=Connector, loc=<0 0 0>, rot=<0 0 0 1>}
 Back Leg L/Toe {type=Connector, loc=<0 0 0>, rot=<0 0 0 1>}
 Back Leg L/Toe Tip {type=Connector, loc=<0 0 0>, rot=<0 0 0 1>}
 Back Leg L/Parent {type=Connector, loc=<0 0 0>, rot=<0 0 0 1>}
 Back Leg L/IK Parent {type=Connector, loc=<0 0 0>, rot=<0 0 0 1>}
 Back Leg L/IK Spaces {type=Connector, loc=<0 0 0>, rot=<0 0 0 1>}
 Back Leg L/Scapula {type=Connector, loc=<0 0 0>, rot=<0 0 0 1>}
 Front Leg L/Start {type=Connector, loc=<0 0 0>, rot=<0 0 0 1>}
 Front Leg L/Mid {type=Connector, loc=<0 0 0>, rot=<0 0 0 1>}
 Front Leg L/End {type=Connector, loc=<0 0 0>, rot=<0 0 0 1>}
 Front Leg L/Ball {type=Connector, loc=<0 0 0>, rot=<0 0 0 1>}
 Front Leg L/Toe {type=Connector, loc=<0 0 0>, rot=<0 0 0 1>}
 Front Leg L/Toe Tip {type=Connector, loc=<0 0 0>, rot=<0 0 0 1>}
 Front Leg L/Parent {type=Connector, loc=<0 0 0>, rot=<0 0 0 1>}
 Front Leg L/IK Parent {type=Connector, loc=<0 0 0>, rot=<0 0 0 1>}
 Front Leg L/IK Spaces {type=Connector, loc=<0 0 0>, rot=<0 0 0 1>}
 Front Leg L/Scapula {type=Connector, loc=<0 0 0>, rot=<0 0 0 1>}
 Back Leg R/Start {type=Connector, loc=<0 0 0>, rot=<0 0 0 1>}
 Back Leg R/Mid {type=Connector, loc=<0 0 0>, rot=<0 0 0 1>}
 Back Leg R/End {type=Connector, loc=<0 0 0>, rot=<0 0 0 1>}
 Back Leg R/Ball {type=Connector, loc=<0 0 0>, rot=<0 0 0 1>}
 Back Leg R/Toe {type=Connector, loc=<0 0 0>, rot=<0 0 0 1>}
 Back Leg R/Toe Tip {type=Connector, loc=<0 0 0>, rot=<0 0 0 1>}
 Back Leg R/Parent {type=Connector, loc=<0 0 0>, rot=<0 0 0 1>}
 Back Leg R/IK Parent {type=Connector, loc=<0 0 0>, rot=<0 0 0 1>}
 Back Leg R/IK Spaces {type=Connector, loc=<0 0 0>, rot=<0 0 0 1>}
 Back Leg R/Scapula {type=Connector, loc=<0 0 0>, rot=<0 0 0 1>}
 Front Leg R/Start {type=Connector, loc=<0 0 0>, rot=<0 0 0 1>}
 Front Leg R/Mid {type=Connector, loc=<0 0 0>, rot=<0 0 0 1>}
 Front Leg R/End {type=Connector, loc=<0 0 0>, rot=<0 0 0 1>}
 Front Leg R/Ball {type=Connector, loc=<0 0 0>, rot=<0 0 0 1>}
 Front Leg R/Toe {type=Connector, loc=<0 0 0>, rot=<0 0 0 1>}
 Front Leg R/Toe Tip {type=Connector, loc=<0 0 0>, rot=<0 0 0 1>}
 Front Leg R/Parent {type=Connector, loc=<0 0 0>, rot=<0 0 0 1>}
 Front Leg R/IK Parent {type=Connector, loc=<0 0 0>, rot=<0 0 0 1>}
 Front Leg R/IK Spaces {type=Connector, loc=<0 0 0>, rot=<0 0 0 1>}
 Front Leg R/Scapula {type=Connector, loc=<0 0 0>, rot=<0 0 0 1>}
 Pin IK Bones/Root {type=Connector, loc=<0 0 0>, rot=<0 0 0 1>}
 Pin IK Bones/Drivers {type=Connector, loc=<0 0 0>, rot=<0 0 0 1>}
 Pin IK Bones/Driven {type=Connector, loc=<0 0 0>, rot=<0 0 0 1>}
]
